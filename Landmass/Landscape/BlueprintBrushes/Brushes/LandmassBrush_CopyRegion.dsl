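; Landmass landscape brush that copies a region of heightmap and chosen weightmap layers into textures and redraws it at the brush.
; CopyLandscapeData forces both layer types on and requests an update; RenderLayer copies each layer into render targets, then converts them to textures, optionally saved assets.
; Once all requested layers are copied, RenderLayer feeds those textures to height and weightmap MIDs carrying falloff, displacement, subdivision and custom parameters, then draws them.
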(event EventBeginPlay)

(event Collision|EventActorBeginOverlap
  (params (OtherActor :type Actor%20Object%20Reference)))

(event EventSetShapeMask
  (params (Texture :type Texture%202D%20Object%20Reference) (Points :type Array%20of%20Vectors) (NewExtents :type Vector%204%20Structure))
  (Utilities|Struct|SetmembersinLandmassEdgeFalloffSettings (Variables|HeightmapSettings|GetEdgeFalloff) Texture)
  (Default|SetupBrushMIDs))

(event EventCopyEdgeMaskParameters
  (params (MID :type Material%20Instance%20Dynamic%20Object%20Reference))
  (Utilities|SetBrushBasicMaterialParams MID)
  (Utilities|SetBrushEdgeFalloffMaterialParams MID (Variables|HeightmapSettings|GetEdgeFalloff))
  (Utilities|SetEdgeFalloffNoiseMaterialParams MID (Variables|HeightmapSettings|GetEdgeFalloffNoise)))

(event Tick|EventCustomTick
  (params (DeltaSeconds :type Float%20%28single-precision%29))
  (bind _weightmap_copy_textures (Variables|LandscapeCopy|GetWeightmap%20Copy%20Textures))
  (bind _height_copy_texture (Variables|LandscapeCopy|GetHeight%20Copy%20Texture))
  (if (Variables|Debug|GetRequestTextureUpdate)
    (if (> (- (Utilities|Time|GetRealTimeSeconds) (Variables|Default|GetTimeTextureUpdateRequested)) 0.5)
      (Default|SetEditorTickEnabled false)
      (Utilities|FlowControl|Sequence
        (:then_0
          (Variables|/Script/Engine.Texture|SetCompressionSettings "TC_VectorDisplacementmap" _height_copy_texture)
          (Variables|/Script/Engine.Texture|SetMipGenSettings "TMGS_NoMipmaps" _height_copy_texture)
          (Rendering|ForceUpdateTexture _height_copy_texture))
        (:then_1
          (bind _keys (Utilities|Map|Keys (Variables|LandscapeCopy|GetWeightmap%20Copy%20Textures)))
          (for _array_element _keys
            (Variables|/Script/Engine.Texture|SetCompressionSettings "TC_VectorDisplacementmap" (Utilities|Map|Find _weightmap_copy_textures _array_element))
            (Variables|/Script/Engine.Texture|SetMipGenSettings "TMGS_NoMipmaps" (Utilities|Map|Find _weightmap_copy_textures _array_element))
            (Rendering|ForceUpdateTexture (Utilities|Map|Find _weightmap_copy_textures _array_element)))))
      (else
        (Development|PrintString "TICKITY")))))

(event Custom|ManagerRenderComplete)

(fn AllocateRTs
  (bind (_x _y) (Utilities|GetBrushRTSize))
  (bind (_x_2 _y_2) (Utilities|GetBrushRTSize))
  (bind _weightmaprts (Variables|Debug|GetWeightmapRTs))
  (Utilities|FlowControl|Sequence
    (:then_0
      (Rendering|CreateAndAssignRenderTarget (Variables|Debug|GetHeightCopyRT) _x_2 _y_2))
    (:then_1
      (for _array_element (Variables|WeightmapSettings|GetWeightmaps%20to%20Copy)
        (LocalVariables|SetCurWM _array_element)
        (if (.ReturnValue (Utilities|Map|Find _weightmaprts _array_element))
          (LocalVariables|SetTempRTRef (Utilities|Map|Find _weightmaprts _array_element))
          (else
            (LocalVariables|SetTempRTRef "")))
        (Rendering|CreateAndAssignRenderTarget (LocalVariables|GetTempRTRef) _x _y "RTF_R8"
          (:Out
            (Utilities|Map|Add (Variables|Debug|GetWeightmapRTs) (LocalVariables|GetCurWM) (LocalVariables|GetTempRTRef))))))))

(fn AssignRenderTargets
  (bind (_renderareaworldtransform _renderareasize _combinedresult) (Utilities|Struct|BreakLandscapeBrushParameters (Variables|Debug|GetBrushRenderParameters)))
  (bind (_renderareaworldtransform_2 _renderareasize_2 _layertype) (Utilities|Struct|BreakLandscapeBrushParameters (Variables|Debug|GetBrushRenderParameters)))
  (bind (_renderareaworldtransform_3 _renderareasize_3 _weightmaplayername) (Utilities|Struct|BreakLandscapeBrushParameters (Variables|Debug|GetBrushRenderParameters)))
  (bind (_renderareaworldtransform_4 _renderareasize_4 _combinedresult_2) (Utilities|Struct|BreakLandscapeBrushParameters (Variables|Debug|GetBrushRenderParameters)))
  (bind _heightmaprendermid (Variables|Debug|GetHeightmapRenderMID))
  (bind _weightmaprendermid (Variables|Debug|GetWeightmapRenderMID))
  (if (Utilities|Enum|Equal(Enum) _layertype "Heightmap")
    (Rendering|Material|SetTextureParameterValue _heightmaprendermid "HeightRT" _combinedresult)
    (Rendering|Material|SetTextureParameterValue _heightmaprendermid "CopiedTexture" (Variables|LandscapeCopy|GetHeight%20Copy%20Texture))
    (else
      (Rendering|Material|SetTextureParameterValue _weightmaprendermid "WeightmapRT" _combinedresult_2)
      (Rendering|Material|SetTextureParameterValue _weightmaprendermid "CopiedTexture" (Utilities|Map|Find (Variables|LandscapeCopy|GetWeightmap%20Copy%20Textures) _weightmaplayername)))))

(fn CopyInputData
  (bind _brushrenderparameters (Variables|Debug|GetBrushRenderParameters))
  (bind (_renderareaworldtransform _renderareasize _weightmaplayername) (Utilities|Struct|BreakLandscapeBrushParameters _brushrenderparameters))
  (bind (_renderareaworldtransform_2 _renderareasize_2 _layertype) (Utilities|Struct|BreakLandscapeBrushParameters (Variables|Debug|GetBrushRenderParameters)))
  (bind _brushrenderparameters_2 (Variables|Debug|GetBrushRenderParameters))
  (bind (_renderareaworldtransform_3 _renderareasize_3 _combinedresult _layertype_2 _weightmaplayername_2) (Utilities|Struct|BreakLandscapeBrushParameters _brushrenderparameters_2))
  (bind _weightmaprts (Variables|Debug|GetWeightmapRTs))
  (bind (_value _returnvalue) (Utilities|Map|Find _weightmaprts _weightmaplayername_2))
  (bind _allvalid (LocalVariables|GetAllValid))
  (bind _allvalid_2 (LocalVariables|GetAllValid))
  (Utilities|FlowControl|Sequence
    (:then_0
      (if (Utilities|Enum|Equal(Enum) _layertype "Heightmap")
        (CallFunction|CopyInputRT (Variables|Debug|GetHeightCopyRT))
        (CallFunction|SaveTextureData (Variables|LandscapeCopy|GetHeight%20Copy%20Texture) (Variables|Debug|GetHeightCopyRT))
        (Utilities|Array|AddUnique (Variables|Debug|GetUpdatedLayers) "Heightmap")
        (elif _returnvalue
          (CallFunction|CopyInputRT _value)
          (bind _output_get (LocalVariables|SetWMTextureRef (Utilities|Map|Find (Variables|LandscapeCopy|GetWeightmap%20Copy%20Textures) _weightmaplayername)))
          (CallFunction|SaveTextureData _output_get (Utilities|Map|Find (Variables|Debug|GetWeightmapRTs) _weightmaplayername))
          (Utilities|Map|Add (Variables|LandscapeCopy|GetWeightmap%20Copy%20Textures) _weightmaplayername (LocalVariables|GetWMTextureRef))
          (Utilities|Array|AddUnique (Variables|Debug|GetUpdatedLayers) _weightmaplayername))))
    (:then_1
      ;[1 1 1] Check if All Requested Layers Rendered
      (LocalVariables|SetAllValid true)
      (Utilities|FlowControl|Sequence
        (:then_0
          (if (not (Utilities|Array|ContainsItem (Variables|Debug|GetUpdatedLayers) "Heightmap"))
            (LocalVariables|SetAllValid false)))
        (:then_1
          (if (> (Utilities|Array|Length (Variables|WeightmapSettings|GetWeightmaps%20to%20Copy)) 0)
            (for _array_element (Variables|WeightmapSettings|GetWeightmaps%20to%20Copy)
              (if (not (Utilities|Array|ContainsItem (Variables|Debug|GetUpdatedLayers) _array_element))
                (LocalVariables|SetAllValid false)))
            (Variables|LandscapeCopy|SetAllLayersCopied _allvalid_2)
            (Variables|Debug|SetRequestCopyData (not _allvalid_2))
            (if _allvalid_2
              _)
            (else
              (Variables|LandscapeCopy|SetAllLayersCopied _allvalid)
              (Variables|Debug|SetRequestCopyData (not _allvalid)))))
        (:then_2
          (if (Variables|LandscapeCopy|GetAllLayersCopied)
            (Variables|BrushSettings|SetAffectsHeightmap (Variables|Debug|GetWas_AffectHeightmap))
            (Variables|BrushSettings|SetAffectsWeightmaps (Variables|Debug|GetWas_AffectWeightmap))))))))

(fn CopyInputRT
  (params (RenderTarget :type Texture%20Render%20Target%202D%20Object%20Reference))
  (bind (_renderareaworldtransform _renderareasize _combinedresult) (Utilities|Struct|BreakLandscapeBrushParameters (Variables|Debug|GetBrushRenderParameters)))
  (bind (_renderareaworldtransform_2 _renderareasize_2 _layertype) (Utilities|Struct|BreakLandscapeBrushParameters (Variables|Debug|GetBrushRenderParameters)))
  (bind (_screenposition _screensize _coordinateposition _coordinatesize) (Landmass|WorldExtentstoLandmassCoordinates (Variables|Debug|GetBrushExtents) (Variables|/Script/LandmassEditor.LandmassManagerBase|GetLandscapeInformation (Variables|Debug|GetBrushManager))))
  (bind (_x _y) (Utilities|GetBrushRTSize))
  (bind _copyweightmid (Variables|Debug|GetCopyWeightMID))
  (bind _copyheightmid (Variables|Debug|GetCopyHeightMID))
  (bind _output_get (LocalVariables|SetIsHeightmap (Utilities|Enum|Equal(Enum) _layertype "Heightmap")))
  (if _output_get
    (Rendering|CreateAndAssignMID _copyheightmid "/Landmass/Landscape/BlueprintBrushes/Materials/Internal/CopyRT_LandscapeUVs_Heightmap.CopyRT_LandscapeUVs_Heightmap"
      (:Out
        (bind _output_get_2 (LocalVariables|SetMID_To_Render _copyheightmid))
        (Rendering|Material|SetScalarParameterValue _output_get_2 "BrushZ" (select (Variables|HeightmapSettings|GetMake%20Height%20Relative) (.z (Transformation|GetActorLocation)) 0.0) :!class /Script/Engine.MaterialInstanceDynamic)))
    (else
      (Rendering|CreateAndAssignMID _copyweightmid "/Landmass/Landscape/BlueprintBrushes/Materials/Internal/CopyRT_LandscapeUVs.CopyRT_LandscapeUVs"
        (:Out
          (LocalVariables|SetMID_To_Render _copyweightmid)))))
  (Rendering|Material|SetTextureParameterValue (LocalVariables|GetMID_To_Render) "RT" _combinedresult)
  (bind (_canvas _size _context) (Rendering|BeginDrawCanvastoRenderTarget RenderTarget))
  (Canvas|DrawMaterial :self _canvas :RenderMaterial (LocalVariables|GetMID_To_Render) :ScreenSize (Math|Vector2D|MakeVector2D (Math|Conversions|ToFloat(Integer) _x) (Math|Conversions|ToFloat(Integer) _y)) :CoordinatePosition (select (Variables|BrushSettings|GetDrawToEntireLandscape) "(X=0.000000,Y=0.000000)" _coordinateposition) :CoordinateSize (select (Variables|BrushSettings|GetDrawToEntireLandscape) "(X=1.000000,Y=1.000000)" _coordinatesize) :Rotation 0.0 :PivotPoint "(X=0.500000,Y=0.500000)")
  (Rendering|EndDrawCanvastoRenderTarget _context)
  (Variables|LandscapeCopy|SetXYOffset (Utilities|GetXYOffset)))

(fn CopyLandscapeData
  (Variables|Debug|SetRequestCopyData true)
  (Variables|LandscapeCopy|SetAllLayersCopied false)
  (Utilities|Array|Clear (Variables|Debug|GetUpdatedLayers))
  (Variables|Debug|SetWas_AffectHeightmap (Variables|BrushSettings|GetAffectsHeightmap))
  (Variables|Debug|SetWas_AffectWeightmap (Variables|BrushSettings|GetAffectsWeightmaps))
  (Variables|BrushSettings|SetAffectsHeightmap true)
  (Variables|BrushSettings|SetAffectsWeightmaps true)
  (Landscape|RequestLandscapeUpdate (Variables|Debug|GetBrushManager) true)
  (bind _aslandmass_brush_manager (Utilities|Casting|CastToLandmassBrushManager (Variables|Debug|GetBrushManager)))
  (Default|BindEventtoLandmassManagerRenderComplete _aslandmass_brush_manager (EventDispatchers|CreateEvent :!event ManagerRenderComplete)))

(fn DrawBrush
  (bind (_renderareaworldtransform _renderareasize _layertype) (Utilities|Struct|BreakLandscapeBrushParameters (Variables|Debug|GetBrushRenderParameters)))
  (if (Utilities|Enum|Equal(Enum) _layertype "Heightmap")
    (Default|DrawBrushMaterial (Variables|Debug|GetHeightmapRenderMID))
    (else
      (Default|DrawBrushMaterial (Variables|Debug|GetWeightmapRenderMID)))))

(fn GetShapeMask
  (return (Utilities|Struct|BreakLandmassEdgeFalloffSettings (Variables|HeightmapSettings|GetEdgeFalloff))))

(fn GetRenderDependencies
  (bind _renderobjects (LocalVariables|GetRenderObjects))
  (Utilities|Array|Clear _renderobjects)
  (Utilities|Array|Add _renderobjects (Variables|LandscapeCopy|GetHeight%20Copy%20Texture))
  (bind _keys (Utilities|Map|Keys (Variables|LandscapeCopy|GetWeightmap%20Copy%20Textures)))
  (for _array_element _keys
    (Utilities|Array|Add (LocalVariables|GetRenderObjects) (Utilities|Map|Find (Variables|LandscapeCopy|GetWeightmap%20Copy%20Textures) _array_element)))
  (return (LocalVariables|GetRenderObjects)))

(fn LandmassEditor
  (Default|LaunchLandmassEditor (Variables|Debug|GetBrushManager) self))

(fn PrintSize
  (Development|PrintString (Utilities|String|ToString(Vector2d) (Utilities|GetXYOffset))))

(fn RenderErosionLayer
  (params (In%20Parameters :type Landscape%20Brush%20Parameters%20Structure) (Extents :type Vector%204%20Structure) (BrushManager :type Landmass%20Brush%20Manager%20Object%20Reference))
  (return false))

(fn RenderLayer
  (params (InParameters :type Landscape%20Brush%20Parameters%20Structure%20%28by%20ref%29))
  (if (Variables|Debug|GetRequestCopyData)
    (CallFunction|AllocateRTs)
    (CallFunction|CopyInputData))
  (if (Variables|LandscapeCopy|GetAllLayersCopied)
    ;[1 0.9016 0.1562] Assign RT to MID
    (CallFunction|AssignRenderTargets)
    ;[0.0349 0.4746 1] Draw the Material
    (CallFunction|DrawBrush)
    ;[0.1699 0.01189 0.25] Complete
    (return)))

(fn SaveTextureData
  (params (Texture :type Texture%202D%20Object%20Reference%20%28by%20ref%29) (RenderTarget :type Texture%20Render%20Target%202D%20Object%20Reference%20%28by%20ref%29))
  (bind _brushrenderparameters (Variables|Debug|GetBrushRenderParameters))
  (bind (_renderareaworldtransform _renderareasize _layertype) (Utilities|Struct|BreakLandscapeBrushParameters _brushrenderparameters))
  (bind _brushrenderparameters_2 (Variables|Debug|GetBrushRenderParameters))
  (bind (_renderareaworldtransform_2 _renderareasize_2 _weightmaplayername) (Utilities|Struct|BreakLandscapeBrushParameters _brushrenderparameters_2))
  (bind _returnvalue (Utilities|String|ToString(Name) _weightmaplayername))
  (bind _returnvalue_2 (Utilities|String|Append "T_WM_" _returnvalue))
  (bind _returnvalue_3 (Utilities|Enum|Equal(Enum) _layertype "Heightmap"))
  (bind _returnvalue_4 (select _returnvalue_3 "T_Heightmap" _returnvalue_2))
  (bind _asset_path (Variables|LandscapeCopy|GetAsset%20Path))
  (bind _returnvalue_5 (Utilities|String|Append _asset_path _returnvalue_4))
  (bind _outpath (Default|ObjectPathAddOuter _returnvalue_5))
  (bind (_returnvalue_6 _object) (Default|IsPathValidSoftObj _outpath))
  (bind _height_copy_texture (Variables|LandscapeCopy|GetHeight%20Copy%20Texture))
  (LocalVariables|SetTextureToSave Texture)
  (LocalVariables|SetRTToSave RenderTarget)
  (if (Variables|LandscapeCopy|GetSave%20Texture%20Separately)
    (if _returnvalue_6
      (bind _astexture_2d (Utilities|Casting|CastToTexture2D _object))
      (Transactions|TransactObject _astexture_2d)
      (EditorScripting|Rendering|ConvertRenderTargettoTexture2DEditorOnly (LocalVariables|GetRTToSave) _astexture_2d)
      (Variables|SetBy-RefVar Texture _astexture_2d)
      (else
        (bind _returnvalue_7 (EditorScripting|Rendering|RenderTarget2DCreateStaticTexture2DEditorOnly (LocalVariables|GetRTToSave) _returnvalue_5 "TC_VectorDisplacementmap"))
        (Variables|SetBy-RefVar Texture _returnvalue_7)
        (Transactions|TransactObject _returnvalue_7)))
    (Variables|/Script/Engine.Texture|SetCompressionSettings "TC_VectorDisplacementmap" _height_copy_texture)
    (Variables|/Script/Engine.Texture|SetMipGenSettings "TMGS_NoMipmaps" _height_copy_texture)
    (Rendering|ForceUpdateTexture _height_copy_texture)
    (else
      (Utilities|IsValid Texture
        (:Is%20Valid
          _)
        (:Is%20Not%20Valid
          (bind _returnvalue_8 (Game|ConstructObjectfromClass "/Script/Engine.Texture2D"))
          (Variables|SetBy-RefVar Texture _returnvalue_8)))
      (Transactions|TransactObject Texture)
      (EditorScripting|Rendering|ConvertRenderTargettoTexture2DEditorOnly (LocalVariables|GetRTToSave) Texture)
      (Variables|/Script/Engine.Texture|SetCompressionSettings "TC_VectorDisplacementmap" Texture)
      (Variables|/Script/Engine.Texture|SetMipGenSettings "TMGS_NoMipmaps" Texture)
      (Rendering|ForceUpdateTexture Texture))))

(fn SetMaterialParams
  (params (In%20MID :type Material%20Instance%20Dynamic%20Object%20Reference) (IsHeightmap :type Boolean))
  (bind _in_mid (Variables|GetIn%20MID))
  (bind _in_mid_2 (Variables|GetIn%20MID))
  (bind _in_mid_3 (Variables|GetIn%20MID))
  (bind _in_mid_4 (Variables|GetIn%20MID))
  (Utilities|IsValid (Variables|Debug|GetBrushManager)
    (:Is%20Valid
      (Utilities|FlowControl|Sequence
        (:then_0
          ;[1 1 1] Basic Brush Parameters
          (Utilities|SetBrushBasicMaterialParams _in_mid_2 (Variables|GetIsHeightmap))
          (Rendering|Material|SetScalarParameterValue _in_mid_2 "ForceUseRegularUVs" (Math|Conversions|ToFloat(Boolean) (Variables|BrushSettings|GetDrawToEntireLandscape)) :!class /Script/Engine.MaterialInstanceDynamic)
          (Rendering|Material|SetScalarParameterValue _in_mid_2 "BrushSize" (Utilities|GetMaterialExtents) :!class /Script/Engine.MaterialInstanceDynamic)
          (Rendering|Material|SetVectorParameterValue _in_mid_3 "BrushPosition" (Math|Conversions|ToLinearColor(Vector) (Math|Conversions|ToVector(Vector2D) (+ (Math|Conversions|ToVector2D(Vector) (Transformation|GetActorLocation)) (Math|Conversions|ToVector2D(Vector) (Math|Transform|TransformDirection (Transformation|GetActorTransform) (Math|Conversions|ToVector(Vector2D) (Variables|LandscapeCopy|GetXYOffset))))))) :!class /Script/Engine.MaterialInstanceDynamic)
          (if (not (Variables|HeightmapSettings|GetMake%20Height%20Relative))
            (Rendering|Material|SetScalarParameterValue _in_mid_3 "BrushZ" 0.0 :!class /Script/Engine.MaterialInstanceDynamic)))
        (:then_1
          ;[1 1 1] Displacement and Falloff
          (Utilities|SetBrushEdgeFalloffMaterialParams (Variables|GetIn%20MID) (Variables|HeightmapSettings|GetEdgeFalloff))
          (Utilities|SetEdgeFalloffNoiseMaterialParams (Variables|GetIn%20MID) (Variables|HeightmapSettings|GetEdgeFalloffNoise))
          (Rendering|Material|SetScalarParameterValue _in_mid "RelativeDepth" (Variables|HeightmapSettings|GetRelative%20Depth) :!class /Script/Engine.MaterialInstanceDynamic)
          (Rendering|Material|SetTextureParameterValue _in_mid "Displacement" (select (Utilities|IsValid (Variables|HeightmapSettings|GetDisplacement) :!class /Script/Engine.KismetSystemLibrary) (Variables|HeightmapSettings|GetDisplacement) "/Engine/ArtTools/RenderToTexture/Textures/T_EV_BlankWhite_01.T_EV_BlankWhite_01"))
          (Rendering|Material|SetScalarParameterValue _in_mid "DisplacementAsAlpha" (Utilities|Select :Index (Variables|HeightmapSettings|DisplacementSettings|GetAlphablend%20Behavior) :NewEnumerator0 1.0 :NewEnumerator1 0.0) :!class /Script/Engine.MaterialInstanceDynamic)
          (Rendering|Material|SetVectorParameterValue _in_mid "DisplacementChannel" (Utilities|Select :Index (Variables|HeightmapSettings|DisplacementSettings|GetTexture%20Channel) :NewEnumerator0 "(R=1.000000,G=0.000000,B=0.000000,A=0.000000)" :NewEnumerator1 "(R=0.000000,G=1.000000,B=0.000000,A=0.000000)" :NewEnumerator2 "(R=0.000000,G=0.000000,B=1.000000,A=0.000000)" :NewEnumerator3 "(R=0.000000,G=0.000000,B=0.000000,A=1.000000)") :!class /Script/Engine.MaterialInstanceDynamic)
          (Rendering|Material|SetScalarParameterValue _in_mid "DisplacementReferencePlane" (Variables|HeightmapSettings|DisplacementSettings|GetHeight%20Reference%20Plane) :!class /Script/Engine.MaterialInstanceDynamic))
        (:then_2
          ;[1 1 1] Apply Custom Parameters: Allows customizing without making a new BP Class
          (Class|LandmassFunctionLibrary|ApplyCustomMaterialParams (Variables|GetIn%20MID) (Variables|HeightmapSettings|CustomParameters|GetCustom%20Scalar%20Params) (Variables|HeightmapSettings|CustomParameters|GetCustom%20Vector%20Params) (Variables|HeightmapSettings|CustomParameters|GetCustom%20Texture%20Params)))
        (:then_3
          (Rendering|Material|SetScalarParameterValue _in_mid_4 "Subdivisions" (Math|Conversions|ToFloat(Integer) (Variables|LandscapeCopy|GetSubdivisions)) :!class /Script/Engine.MaterialInstanceDynamic)
          (Rendering|Material|SetScalarParameterValue _in_mid_4 "Index" (Math|Conversions|ToFloat(Integer) (Variables|LandscapeCopy|GetIndex)) :!class /Script/Engine.MaterialInstanceDynamic)
          (Rendering|Material|SetScalarParameterValue _in_mid_4 "Use Shared Edges" (Math|Conversions|ToFloat(Boolean) (Variables|LandscapeCopy|GetUse%20Shared%20Edges)) :!class /Script/Engine.MaterialInstanceDynamic))))))

(fn SetupBrushMIDs
  (Utilities|FlowControl|Sequence
    (:then_0
      (Rendering|CreateAndAssignMID (Variables|Debug|GetHeightmapRenderMID) (Variables|HeightmapSettings|GetHeightMaterial)
        (:Out
          (Default|SetMaterialParams (Variables|Debug|GetHeightmapRenderMID) true)
          (Utilities|SetupBrushPreviewShape (Variables|HeightmapSettings|GetEdgeFalloff) (Variables|HeightmapSettings|GetEdgeFalloffNoise)))))
    (:then_1
      (Rendering|CreateAndAssignMID (Variables|Debug|GetWeightmapRenderMID) (Variables|WeightmapSettings|GetWeightmapMaterial)
        (:Out
          (Default|SetMaterialParams (Variables|Debug|GetWeightmapRenderMID)))))))

(fn ConstructionScript
  (Utilities|SkipifTransientObject self
    (:Exectute
      (Variables|Debug|SetWeightmapLayers (Variables|WeightmapSettings|GetWeightmaps%20to%20Copy))
      (Default|SetupBrushMIDs)
      (Default|SetCaptureBoundaryNormals (Variables|Debug|GetBrushManager) (Variables|LandscapeCopy|GetUse%20Capture%20Boundary%20Normals))
      (Default|RequestUpdatefromBrush (Variables|Debug|GetBrushManager) self))))
