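; Landmass Blueprint landscape brush that runs a GPU erosion simulation on heightmaps and weightmaps.
; Tick steps through erosion keyframes in several iterations per frame: particle erosion, slope smoothing, water levelling or noise injection.
; Passes draw materials into ping-ponged render targets, configured via MIDs and MPC_Erosion; particles use scene-captured instanced meshes to drive weightmaps.
; On completion it saves the result as a relative texture and requests a landscape update.
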(event EventInitialize
  (params (InLandscapeTransform :type Transform%20%28by%20ref%29) (InLandscapeSize :type Int%20Point%20Structure%20%28by%20ref%29) (InLandscapeRenderTargetSize :type Int%20Point%20Structure%20%28by%20ref%29))
  (Variables|Debug|SetLandscapeTransform InLandscapeTransform)
  (Variables|Debug|SetLandscape%20Size InLandscapeSize)
  (Variables|Debug|SetRT%20Resolution InLandscapeRenderTargetSize)
  (Initialization|InitializeSize)
  (Development|PrintString "Erosion Brush Initialize" false))

(event EventTick
  (params (DeltaSeconds :type Float%20%28single-precision%29))
  (bind _current_iteration (Variables|Default|GetCurrent%20Iteration))
  (if (Variables|Default|GetSimulating)
    ;[0.3248 1 0.5712] Allow Multiple Ticks per Frame
    (for _ (range (Variables|ErosionSettings|GetIterations%20Per%20Tick))
      (if (Variables|Default|GetSimulating)
        ;[1 1 1] Check Current Iteration against Total
        (if (< (Variables|Default|GetCurrent%20Iteration) (Variables|Default|GetTotal%20Iterations))
          ;[0.08541 0.434 1] Run Current Erosion Key and then Compute Normals
          (Utilities|FlowControl|Sequence
            (:then_0
              (switch Utilities|FlowControl|Switch|SwitchonE_Landmass_CustomBrush_ErosionModes (Utilities|Struct|BreakLandmassErosionKeyframe (Variables|Debug|GetCurrentErosionKeyframe))
                (:NewEnumerator0
                  (CallFunction|ErosionParticles))
                (:NewEnumerator3
                  (CallFunction|ErosionSlopeSmoothing))
                (:NewEnumerator4
                  (CallFunction|ErosionWaterLevelling))
                (:NewEnumerator1
                  (CallFunction|ErosionAddNoise))))
            (:then_1
              (Utilities|ComputeNormal)
              (Utilities|UpdatePreviewMIDs)
              (Math|Integer|IncrementInt _current_iteration
                (:%20%20
                  (Rendering|Material|SetScalarParameterValue "/Landmass/Landscape/BlueprintBrushes/MPC/MPC_Erosion.MPC_Erosion" "idx" (Math|Conversions|ToFloat(Integer) _current_iteration) :!class /Script/Engine.KismetMaterialLibrary)))))
          (else
            ;[1 0.7365 0.115] Check if there is a Next Erosion Key to Simulate
            (Math|Integer|IncrementInt (Variables|Default|GetIteration%20Key%20Index)
              (:%20%20
                (if (<= (Variables|Default|GetIteration%20Key%20Index) (Utilities|Array|LastIndex (Variables|ErosionSettings|GetErosion%20Keyframes)))
                  (Variables|Default|SetCurrent%20Iteration 0)
                  (Development|PrintString "Erosion on to Next Key")
                  (Initialization|InitializeErosionValues (Utilities|Array|Get(aref) (Variables|ErosionSettings|GetErosion%20Keyframes) (Variables|Default|GetIteration%20Key%20Index)))
                  (else
                    (CallFunction|CompleteSimulation)))))))))))

(event Custom|Update%20Erosion%20From%20Mask
  (Utilities|IsValid (Variables|ErosionSettings|GetMask)
    (:Is%20Valid
      (bind _aslandmass_brush_manager (Utilities|Casting|CastToLandmassBrushManager (Variables|/Script/LandmassEditor.LandmassActor|GetBrushManager (Variables|ErosionSettings|GetMask))))
      (if (not (Variables|/Landmass/Landscape/BlueprintBrushes/LandmassBrushManager.LandmassBrushManager_C|GetEnable%20Fast%20Preview%20Mode _aslandmass_brush_manager))
        (Landscape|RequestLandscapeUpdate true)))))

(event Custom|Bind%20Mask%20Event
  (bind _mask (Variables|ErosionSettings|GetMask))
  (Utilities|IsValid _mask
    (:Is%20Valid
      (Landmass|BindEventtoOnBrushUpdated _mask (delegate Custom|Update%20Erosion%20From%20Mask)))))

(event Custom|Complete%20Simulation
  (bind (_seconds _partialseconds) (Utilities|Time|GetAccurateRealTime))
  ;[0.04907 0.6562 0] Simulation Complete
  (Variables|Default|SetEnd%20Time (+ (Math|Conversions|ToFloat(Integer) _seconds) _partialseconds))
  (Variables|Debug|SetIs%20Simulation%20Done true)
  (Development|PrintString (Utilities|String|Append "Erosion Complete in: " (Utilities|String|ToString(Float) (- (Variables|Default|GetEnd%20Time) (Variables|Default|GetStart%20Time))) " seconds"))
  (Variables|Default|SetSimulating false)
  (CallFunction|SaveResultAsRelativeTexture)
  (Debug|DisablePreview)
  (Default|CallErosionSimulationComplete)
  (Landscape|RequestLandscapeUpdate true))

(event EventSetShapeMask
  (params (Texture :type Texture%202D%20Object%20Reference) (Points :type Array%20of%20Vectors) (NewExtents :type Vector%204%20Structure)))

(event EventCopyEdgeMaskParameters
  (params (MID :type Material%20Instance%20Dynamic%20Object%20Reference)))

(fn CopyTargettoCBRT
  (bind _copy_export_target (Variables|Export|GetCopy/Export%20Target))
  (bind _weightmaprts (Variables|Debug|RenderTargets|GetWeightmapRTs))
  (bind (_value _returnvalue) (Utilities|Map|Find _weightmaprts _copy_export_target))
  (bind _copy_rt_mid (Variables|Debug|MIDs|GetCopy%20RT%20MID))
  (if (== (Variables|Export|GetCopy/Export%20Target) "Heightmap")
    (Rendering|Material|SetTextureParameterValue _copy_rt_mid "RT" (Utilities|HeightPingPong))
    (elif _returnvalue
      (Rendering|Material|SetTextureParameterValue _copy_rt_mid "RT" _value)))
  (Rendering|DrawMaterialtoRenderTarget (Variables|Export|GetContent%20Browser%20RT) (Variables|Debug|MIDs|GetCopy%20RT%20MID)))

(fn DisablePreview
  (Development|ExecuteConsoleCommand "showflag.landscape 1")
  (Rendering|SetVisibility (Variables|Default|GetLandscape%20ISMC) false false)
  (bind _returnvalue (Landmass|GetLandscape))
  (Landscape|Runtime|Material|SetLandscapeMaterialScalarParameterValue _returnvalue "EnableLandmassPreview" 0.0))

(fn Erosion_AddNoise
  (bind (_read _write) (Utilities|HeightPingPong))
  ;[1 1 1] Add Noise
  (Rendering|Material|SetTextureParameterValue (Variables|Debug|MIDs|GetAdd%20Noise%20MID) "HeightRT" _read)
  (Rendering|ClearRenderTarget2D _write)
  (Rendering|DrawMaterialtoRenderTarget _write (Variables|Debug|MIDs|GetAdd%20Noise%20MID))
  (Math|Integer|IncrementInt (Variables|Default|GetHeight%20idx)))

(fn Erosion_Particles
  (bind _currenterosionkeyframe (Variables|Debug|GetCurrentErosionKeyframe))
  (bind _particleerosion_7_0a79d8e74ce118f7f751d0be5d6ff267 (Utilities|Struct|BreakLandmassErosionKeyframe _currenterosionkeyframe))
  (bind _resetparticleseveryniterations_2_36ed184841e541079ddf9ebf4664e45c (Utilities|Struct|BreakLandmassParticleErosionSettings _particleerosion_7_0a79d8e74ce118f7f751d0be5d6ff267))
  (bind _particle_position_seed_mid (Variables|Debug|MIDs|GetParticle%20Position%20Seed%20MID))
  (bind (_read _write) (Utilities|VelocityPingPong))
  (bind (_read_2 _write_2) (Utilities|PositionPingPong))
  (bind (_read_3 _write_3) (Utilities|VelocityPingPong))
  (bind (_read_4 _write_4) (Utilities|HeightPingPong))
  (bind _scatter_velocity_mid (Variables|Debug|MIDs|GetScatter%20Velocity%20MID))
  (bind _scatter_position_mid (Variables|Debug|MIDs|GetScatter%20Position%20MID))
  (bind _scatter_height_erosion_mid (Variables|Debug|MIDs|GetScatter%20Height%20Erosion%20MID))
  (bind _weightmaprts (Variables|Debug|RenderTargets|GetWeightmapRTs))
  (bind _particle_preview_mid (Variables|Debug|MIDs|GetParticle%20Preview%20MID))
  (bind _particle_weightmap_mid (Variables|Debug|MIDs|GetParticle%20Weightmap%20MID))
  (Utilities|FlowControl|Sequence
    (:then_0
      ;[1 1 1] Scatter Velocity
      (Rendering|Material|SetTextureParameterValue _scatter_velocity_mid "HeightRT" (Utilities|HeightPingPong))
      (Rendering|Material|SetTextureParameterValue _scatter_velocity_mid "PositionRT" (Utilities|PositionPingPong))
      (Rendering|Material|SetTextureParameterValue _scatter_velocity_mid "VelocityRT" _read)
      (Rendering|ClearRenderTarget2D _write)
      (Rendering|DrawMaterialtoRenderTarget _write (Variables|Debug|MIDs|GetScatter%20Velocity%20MID))
      (Math|Integer|IncrementInt (Variables|Default|GetVelocity%20Idx)))
    (:then_2
      ;[1 1 1] Scatter Position
      (Rendering|Material|SetTextureParameterValue _scatter_position_mid "HeightRT" (Utilities|HeightPingPong))
      (Rendering|Material|SetTextureParameterValue _scatter_position_mid "VelocityRT" (Utilities|VelocityPingPong))
      (Rendering|Material|SetTextureParameterValue _scatter_position_mid "PositionRT" _read_2)
      (Rendering|ClearRenderTarget2D _write_2)
      (Rendering|DrawMaterialtoRenderTarget _write_2 (Variables|Debug|MIDs|GetScatter%20Position%20MID))
      (Math|Integer|IncrementInt (Variables|Default|GetPosition%20Idx)))
    (:then_4
      ;[1 1 1] Capture Particles to Render Target
      (Rendering|Material|SetMaterial (Variables|Default|GetScatter%20Particles%20ISMC) 0 (Variables|Debug|MIDs|GetParticle%20Render%20MID))
      (Rendering|Material|SetTextureParameterValue _particle_preview_mid "VelocityRT" _read_3)
      (Rendering|Material|SetTextureParameterValue _particle_preview_mid "VelocityRTPrev" _write_3)
      (Rendering|Material|SetTextureParameterValue _particle_preview_mid "PositionRT" (Utilities|PositionPingPong))
      (Rendering|SetVisibility (Variables|Default|GetScatter%20Particles%20ISMC) true false)
      (Rendering|SceneCapture|CaptureScene (Variables|Default|GetScatterMeshSceneCapture) :!class /Script/Engine.SceneCaptureComponent2D)
      (Rendering|Material|SetMaterial (Variables|Default|GetScatter%20Particles%20ISMC) 0 (select (Variables|Debug|GetDisplay%20Particle%20Render%20Mat) (Variables|Debug|MIDs|GetParticle%20Render%20MID) (Variables|Debug|MIDs|GetParticle%20Preview%20MID)))
      (Rendering|SetVisibility (Variables|Default|GetScatter%20Particles%20ISMC) (Variables|Debug|GetShow%20Water) false))
    (:then_6
      ;[1 1 1] Erode Terrain using Particles
      (Rendering|Material|SetTextureParameterValue _scatter_height_erosion_mid "HeightRT" (Utilities|HeightPingPong))
      (Rendering|Material|SetTextureParameterValue _scatter_height_erosion_mid "PositionRT" (Utilities|PositionPingPong))
      (Rendering|Material|SetTextureParameterValue _scatter_height_erosion_mid "VelocityRT" (Utilities|VelocityPingPong))
      (Rendering|DrawMaterialtoRenderTarget _write_4 (Variables|Debug|MIDs|GetScatter%20Height%20Erosion%20MID))
      (Math|Integer|IncrementInt (Variables|Default|GetHeight%20idx)
        (:%20%20
          ;; Reset Iterations
          (if (> _resetparticleseveryniterations_2_36ed184841e541079ddf9ebf4664e45c 0)
            (if (== (Math|Integer|%25%28Integer%29 (Variables|Default|GetCurrent%20Iteration) _resetparticleseveryniterations_2_36ed184841e541079ddf9ebf4664e45c) 0)
              (Rendering|ClearRenderTarget2D (Variables|Debug|RenderTargets|GetVelocity%20RT%20A) "(R=0.000000,G=0.000000,B=0.000000,A=0.000000)")
              (Rendering|ClearRenderTarget2D (Variables|Debug|RenderTargets|GetVelocity%20RT%20B) "(R=0.000000,G=0.000000,B=0.000000,A=0.000000)")
              (Rendering|ClearRenderTarget2D (Variables|Debug|RenderTargets|GetPosition%20RT%20A))
              (Rendering|ClearRenderTarget2D (Variables|Debug|RenderTargets|GetPosition%20RT%20B))
              (Rendering|Material|SetTextureParameterValue (Variables|Debug|MIDs|GetParticle%20Position%20Seed%20MID) "HeightRT" (Utilities|HeightPingPong))
              (Rendering|DrawMaterialtoRenderTarget (Variables|Debug|RenderTargets|GetPosition%20RT%20A) _particle_position_seed_mid)
              (Rendering|DrawMaterialtoRenderTarget (Variables|Debug|RenderTargets|GetPosition%20RT%20B) _particle_position_seed_mid)
              (Variables|Default|SetPosition%20Idx 0)
              (Variables|Default|SetVelocity%20Idx 0))))))
    (:then_8
      (for _array_element (Utilities|Struct|BreakLandmassParticleErosionSettings (Utilities|Struct|BreakLandmassErosionKeyframe (Variables|Debug|GetCurrentErosionKeyframe)))
        (Rendering|Material|SetScalarParameterValue _particle_weightmap_mid "Deposition" (.DepositionStrength_9_A0C69D6145F47973775021AB8749B742 (Utilities|Struct|BreakErodeWeightmap _array_element)) :!class /Script/Engine.MaterialInstanceDynamic)
        (Rendering|Material|SetScalarParameterValue _particle_weightmap_mid "Erosion" (.ErosionStrength_10_CF946B9A4ADA64BEEB4E8DA6A2CEA428 (Utilities|Struct|BreakErodeWeightmap _array_element)) :!class /Script/Engine.MaterialInstanceDynamic)
        (Rendering|Material|SetScalarParameterValue _particle_weightmap_mid "Velocity" (.VelocityStrength_14_FC9BD95E42791F6B55C07BAAF882F358 (Utilities|Struct|BreakErodeWeightmap _array_element)) :!class /Script/Engine.MaterialInstanceDynamic)
        (if (.ReturnValue (Utilities|Map|Find _weightmaprts (Utilities|String|StringToName (Utilities|Struct|BreakErodeWeightmap _array_element))))
          (Rendering|DrawMaterialtoRenderTarget (Utilities|Map|Find _weightmaprts (Utilities|String|StringToName (Utilities|Struct|BreakErodeWeightmap _array_element))) (Variables|Debug|MIDs|GetParticle%20Weightmap%20MID)))))))

(fn Erosion_SlopeSmoothing
  (bind (_read _write) (Utilities|HeightPingPong))
  (Utilities|FlowControl|Sequence
    (:then_0
      ;[1 1 1] Smooth Terrain using Slope Thresholds
      (Rendering|Material|SetTextureParameterValue (Variables|Debug|MIDs|GetSlope%20Smoothing%20MID) "HeightRT" _read)
      (Rendering|DrawMaterialtoRenderTarget _write (Variables|Debug|MIDs|GetSlope%20Smoothing%20MID))
      (Math|Integer|IncrementInt (Variables|Default|GetHeight%20idx)))))

(fn Erosion_WaterLevelling
  (bind (_read _write) (Utilities|HeightPingPong))
  (bind (_read_2 _write_2) (Utilities|HeightPingPong))
  (bind _water_levelling_weightmap_mid (Variables|Debug|MIDs|GetWater%20Levelling%20Weightmap%20MID))
  (bind _weightmaprts (Variables|Debug|RenderTargets|GetWeightmapRTs))
  (Utilities|FlowControl|Sequence
    (:then_0
      ;[1 1 1] Smooth Terrain using Slope Thresholds
      (Rendering|Material|SetTextureParameterValue (Variables|Debug|MIDs|GetWater%20Levelling%20MID) "HeightRT" _read_2)
      (Rendering|ClearRenderTarget2D _write_2)
      (Rendering|DrawMaterialtoRenderTarget _write_2 (Variables|Debug|MIDs|GetWater%20Levelling%20MID))
      (Math|Integer|IncrementInt (Variables|Default|GetHeight%20idx)))
    (:then_1
      (for _array_element (Utilities|Struct|BreakLandmassWaterLevellingSettings (Utilities|Struct|BreakLandmassErosionKeyframe (Variables|Debug|GetCurrentErosionKeyframe)))
        (if (.ReturnValue (Utilities|Map|Find _weightmaprts (Utilities|String|StringToName (Utilities|Struct|BreakErodeWeightmap _array_element))))
          (Rendering|Material|SetTextureParameterValue _water_levelling_weightmap_mid "HeightRT" _read)
          (Rendering|Material|SetTextureParameterValue _water_levelling_weightmap_mid "HeightRTPrev" _write)
          (Rendering|DrawMaterialtoRenderTarget (Utilities|Map|Find _weightmaprts (Utilities|String|StringToName (Utilities|Struct|BreakErodeWeightmap _array_element))) (Variables|Debug|MIDs|GetWater%20Levelling%20Weightmap%20MID)))))))

(fn GetShapeMask
  (return ""))

(fn GetRenderDependencies
  (return 0))

(fn InitializeBrush
  (bind (_x _y) (Utilities|Struct|BreakIntPoint (Variables|Debug|GetRT%20Resolution)))
  (bind (_x_2 _y_2) (Utilities|Struct|BreakIntPoint (Variables|Debug|GetRT%20Resolution)))
  (bind _currentweightmapsettings (LocalVariables|GetCurrentWeightMapSettings))
  (bind _weightmapname_3_b69a53a640837a50633272bf0b14b24b (Utilities|Struct|BreakErodeWeightmap _currentweightmapsettings))
  (bind (_x_3 _y_3) (Utilities|Struct|BreakIntPoint (Variables|Debug|GetRT%20Resolution)))
  (bind (_x_4 _y_4) (Utilities|Struct|BreakIntPoint (Variables|Debug|GetRT%20Resolution)))
  (bind _returnvalue (Utilities|String|StringToName _weightmapname_3_b69a53a640837a50633272bf0b14b24b))
  (bind _weightmaprts (Variables|Debug|RenderTargets|GetWeightmapRTs))
  (bind (_value _returnvalue_2) (Utilities|Map|Find _weightmaprts _returnvalue))
  (Utilities|FlowControl|Sequence
    (:then_0
      (Rendering|CreateAndAssignRenderTarget (Variables|Debug|RenderTargets|GetHeight%20RT%20A) _x_3 _y_3 "RTF_RG32f"
        (:Out
          (Rendering|CreateAndAssignRenderTarget (Variables|Debug|RenderTargets|GetHeight_Output_RG8) _x_2 _y_2
            (:Out
              (Rendering|CreateAndAssignRenderTarget (Variables|Debug|RenderTargets|GetLandscape%20Height%20Copy%20RT) _x_4 _y_4 "RTF_RG8") ; This RT is a copy of the input Landscape in order to reset and composite
            )) ; This RT is used to apply the erosion to landscape and return it.
        )) ; Needed to save the Landscape RT unpacked so it is available for later simulation.
    )
    (:then_1
      (Rendering|CreateAndAssignMID (Variables|Debug|MIDs|GetApply%20Erosion%20to%20Landscape%20MID) (Variables|ErosionSettings|GetErosion%20Heightmap%20Mask%20Material) "Erosion_ComputeNormals_MID"
        (:Out
          (Rendering|CreateAndAssignMID (Variables|Debug|MIDs|GetApply%20Weightmap%20MID) (Variables|ErosionSettings|GetErosion%20Weightmap%20Mask%20Material) "Erosion_SlopeSmoothing_MID"
            (:Out
              (Rendering|CreateAndAssignMID (Variables|Debug|MIDs|GetCopy%20RT%20MID) "/Landmass/Landscape/ErosionBrush/Materials/Packing/ErosionBrush_CopyRT.ErosionBrush_CopyRT" "Erosion_ComputeNormals_MID"
                (:Out
                  (Utilities|ApplyMaskActor))))))) ; This MID is needed to Apply saved data to Landscape.
    )
    (:then_2
      (Utilities|Array|Clear (LocalVariables|GetWeightmapNames))
      (for _array_element (Variables|ErosionSettings|GetErosion%20Keyframes)
        (if (Utilities|Enum|Equal(Enum) (Utilities|Struct|BreakLandmassErosionKeyframe _array_element) "NewEnumerator0")
          (if (> (Utilities|Array|Length (Utilities|Struct|BreakLandmassParticleErosionSettings (.ParticleErosion_7_0A79D8E74CE118F7F751D0BE5D6FF267 (Utilities|Struct|BreakLandmassErosionKeyframe _array_element)))) 0)
            (for _array_element_2 (Utilities|Struct|BreakLandmassParticleErosionSettings (.ParticleErosion_7_0A79D8E74CE118F7F751D0BE5D6FF267 (Utilities|Struct|BreakLandmassErosionKeyframe _array_element)))
              (LocalVariables|SetCurrentWeightMapSettings _array_element_2)
              (if (!= _returnvalue "")
                (Utilities|Array|AddUnique (LocalVariables|GetWeightmapNames) _returnvalue)
                (if _returnvalue_2
                  (LocalVariables|SetTempRTRef _value)
                  (else
                    (LocalVariables|SetTempRTRef "")))
                (Rendering|CreateAndAssignRenderTarget (LocalVariables|GetTempRTRef) _x _y "RTF_R16f"
                  (:Out
                    (Utilities|Map|Add (Variables|Debug|RenderTargets|GetWeightmapRTs) _returnvalue (LocalVariables|GetTempRTRef))
                    (bind _returnvalue_3 (Landmass|GetLandscape))
                    (Landscape|Runtime|Material|SetLandscapeMaterialTextureParameterValue _returnvalue_3 _returnvalue (LocalVariables|GetTempRTRef)))) ; Needed to save the Landscape RT unpacked so it is available for later simulation.
              )))
          (elif (Utilities|Enum|Equal(Enum) (Utilities|Struct|BreakLandmassErosionKeyframe _array_element) "NewEnumerator4")
            (if (> (Utilities|Array|Length (Utilities|Struct|BreakLandmassWaterLevellingSettings (.WaterLevelling_18_4A26315248B5E070DA9AA887FB36AEC0 (Utilities|Struct|BreakLandmassErosionKeyframe _array_element)))) 0)
              (for _array_element_3 (Utilities|Struct|BreakLandmassWaterLevellingSettings (.WaterLevelling_18_4A26315248B5E070DA9AA887FB36AEC0 (Utilities|Struct|BreakLandmassErosionKeyframe _array_element)))
                (LocalVariables|SetCurrentWeightMapSettings _array_element_3))))))
      ;[1 1 1] Remove an Weightmaps not currently Referenced
      (bind _keys (Utilities|Map|Keys (Variables|Debug|RenderTargets|GetWeightmapRTs)))
      (for _array_element_4 _keys
        (if (not (>= (Utilities|Array|FindItem (LocalVariables|GetWeightmapNames) _array_element_4) 0))
          (Utilities|Map|Remove (Variables|Debug|RenderTargets|GetWeightmapRTs) _array_element_4)))
      (bind _keys_2 (Utilities|Map|Keys (Variables|Debug|RenderTargets|GetWeightmapRTs)))
      (bind _output_get (Variables|Settings|SetAffectedWeightmapLayers _keys_2))
      (Variables|Settings|SetAffectWeightmap (> (Utilities|Array|Length _output_get) 0)))))

(fn InitializeErosionValues
  (params (Erosion%20Key :type Landmass%20Erosion%20Keyframe%20Structure))
  (bind _currenterosionkeyframe (Variables|Debug|GetCurrentErosionKeyframe))
  (bind _waterlevelling_18_4a26315248b5e070da9aa887fb36aec0 (Utilities|Struct|BreakLandmassErosionKeyframe _currenterosionkeyframe))
  (bind (_startingwateramount_5_6662d7234e14e20ac900cbb1168ed16d _rainperiteration_27_6fa2622d48a07f8ad76adc8a52d2335e _solubility_9_66a62d6f47fb09a92282f9998fa8e597 _evaporation_11_c7bafcb548bbb82468aad2b764749543) (Utilities|Struct|BreakLandmassWaterLevellingSettings _waterlevelling_18_4a26315248b5e070da9aa887fb36aec0))
  (bind _currenterosionkeyframe_2 (Variables|Debug|GetCurrentErosionKeyframe))
  (bind _slopesmoothing_17_239f609f40e5e0976999a5a4b7c63485 (Utilities|Struct|BreakLandmassErosionKeyframe _currenterosionkeyframe_2))
  (bind (_smoothslopesgreaterthan_18_b1c0f90747c99e2d459d01982e31c616 _smoothslopeslessthan_20_ad0c525248c27f272b99899db760361a _opacity_30_214597f847b9e91587d7c288f99ce3aa) (Utilities|Struct|BreakLandmassSlopeSmoothingSettings _slopesmoothing_17_239f609f40e5e0976999a5a4b7c63485))
  (bind _currenterosionkeyframe_3 (Variables|Debug|GetCurrentErosionKeyframe))
  (bind _noiseinjection_19_a70da365424143da3f5375985ef7048b (Utilities|Struct|BreakLandmassErosionKeyframe _currenterosionkeyframe_3))
  (bind (_noiseheightmeters_20_74ba1f384b27ba15040a558b6d44a4e3 _noisescalemeters_21_c8ee95774aa7dca911e3f59e537a0cbb _noiseoctaves_4_d613165642c15bc55548b69690950d03 _noisepersistence_11_fa6ffe4e40449bf51d5ddf9c79362a04 _noiseseed_14_ca1226e144c86e22d65c449b9fff7c91 _turbulence_16_f6422936458b63815688e498d8f934d5 _noisefloor0_1_22_6deae6d549095a9ca2c08daab0f22af5 _slopemask_25_f63fef4841c61d02307971af848d0f4b) (Utilities|Struct|BreakLandmassErosionNoiseInjection _noiseinjection_19_a70da365424143da3f5375985ef7048b))
  (bind (_iterations_35_3293851a4544dbc48b67f9aca08cfb63 _material_19_9623e74447d05ea3c529e584a7638d11) (Utilities|Struct|BreakLandmassErosionCustomMaterial (Utilities|Struct|BreakLandmassErosionKeyframe (Variables|Debug|GetCurrentErosionKeyframe))))
  (bind (_scalarparameters_24_30775f0f4d9c8b634089f9b367097495 _vectorparameters_36_cefaeee94c49cc8b1958458614c3c8cc _textureparameters_33_1825952e489caf47f454d089a554972f) (Utilities|Struct|BreakLandmassErosionCustomMaterial (Utilities|Struct|BreakLandmassErosionKeyframe (Variables|Debug|GetCurrentErosionKeyframe))))
  (bind _currenterosionkeyframe_4 (Variables|Debug|GetCurrentErosionKeyframe))
  (bind _particleerosion_7_0a79d8e74ce118f7f751d0be5d6ff267 (Utilities|Struct|BreakLandmassErosionKeyframe _currenterosionkeyframe_4))
  (bind (_startingwateramount_5_6662d7234e14e20ac900cbb1168ed16d_2 _deltat_7_6fa2622d48a07f8ad76adc8a52d2335e _solubility_9_66a62d6f47fb09a92282f9998fa8e597_2 _evaporation_11_c7bafcb548bbb82468aad2b764749543_2 _velocitydamping_27_b1c0f90747c99e2d459d01982e31c616 _velocityadvection_16_390741a84fa9ab14da3866a3c931dbef) (Utilities|Struct|BreakLandmassParticleErosionSettings _particleerosion_7_0a79d8e74ce118f7f751d0be5d6ff267))
  (bind _currenterosionkeyframe_5 (Variables|Debug|GetCurrentErosionKeyframe))
  (bind _particleerosion_7_0a79d8e74ce118f7f751d0be5d6ff267_2 (Utilities|Struct|BreakLandmassErosionKeyframe _currenterosionkeyframe_5))
  (bind (_erodeweightmaps_31_82c4f771449486402aaa308549fecf1b _resolutiondivisor_30_9cec792142f360e868d2129d125b4a4d) (Utilities|Struct|BreakLandmassParticleErosionSettings _particleerosion_7_0a79d8e74ce118f7f751d0be5d6ff267_2))
  (bind _rt_resolution (Variables|Debug|GetRT%20Resolution))
  (bind (_x _y) (Utilities|Struct|BreakIntPoint _rt_resolution))
  (bind _currenterosionkeyframe_6 (Variables|Debug|GetCurrentErosionKeyframe))
  (bind _particleerosion_7_0a79d8e74ce118f7f751d0be5d6ff267_3 (Utilities|Struct|BreakLandmassErosionKeyframe _currenterosionkeyframe_6))
  (bind (_erodeweightmaps_31_82c4f771449486402aaa308549fecf1b_2 _resolutiondivisor_30_9cec792142f360e868d2129d125b4a4d_2) (Utilities|Struct|BreakLandmassParticleErosionSettings _particleerosion_7_0a79d8e74ce118f7f751d0be5d6ff267_3))
  (bind (_x_2 _y_2) (Utilities|Struct|BreakIntPoint (Math|Conversions|ToIntPoint(Vector2D) (/ (Variables|Debug|GetRT%20Resolution) (Math|Conversions|ToFloat(Integer) _resolutiondivisor_30_9cec792142f360e868d2129d125b4a4d)))))
  (bind _returnvalue (Math|Conversions|ToFloat(Integer) _resolutiondivisor_30_9cec792142f360e868d2129d125b4a4d))
  (bind _rt_resolution_2 (Variables|Debug|GetRT%20Resolution))
  (bind _returnvalue_2 (/ _rt_resolution_2 _returnvalue))
  (bind _returnvalue_3 (Math|Conversions|ToIntPoint(Vector2D) _returnvalue_2))
  (bind (_x_3 _y_3) (Utilities|Struct|BreakIntPoint _returnvalue_3))
  (bind (_x_4 _y_4) (Utilities|Struct|BreakIntPoint _returnvalue_3))
  (bind (_x_5 _y_5) (Utilities|Struct|BreakIntPoint _returnvalue_3))
  (bind (_x_6 _y_6) (Utilities|Struct|BreakIntPoint _returnvalue_3))
  (bind (_read _write) (Utilities|HeightPingPong))
  (bind _water_levelling_weightmap_mid (Variables|Debug|MIDs|GetWater%20Levelling%20Weightmap%20MID))
  (bind _water_levelling_seed_mid (Variables|Debug|MIDs|GetWater%20Levelling%20Seed%20MID))
  (bind _add_noise_mid (Variables|Debug|MIDs|GetAdd%20Noise%20MID))
  (bind _particle_weightmap_mid (Variables|Debug|MIDs|GetParticle%20Weightmap%20MID))
  (Variables|Debug|SetCurrentErosionKeyframe Erosion%20Key)
  (Utilities|FlowControl|Sequence
    (:then_0
      (switch Utilities|FlowControl|Switch|SwitchonE_Landmass_CustomBrush_ErosionModes (Utilities|Struct|BreakLandmassErosionKeyframe (Variables|Debug|GetCurrentErosionKeyframe))
        (:NewEnumerator0
          (Rendering|Material|SetScalarParameterValue "/Landmass/Landscape/BlueprintBrushes/MPC/MPC_Erosion.MPC_Erosion" "StartWater" _startingwateramount_5_6662d7234e14e20ac900cbb1168ed16d_2 :!class /Script/Engine.KismetMaterialLibrary)
          (Rendering|Material|SetScalarParameterValue "/Landmass/Landscape/BlueprintBrushes/MPC/MPC_Erosion.MPC_Erosion" "deltaT" _deltat_7_6fa2622d48a07f8ad76adc8a52d2335e :!class /Script/Engine.KismetMaterialLibrary)
          (Rendering|Material|SetScalarParameterValue "/Landmass/Landscape/BlueprintBrushes/MPC/MPC_Erosion.MPC_Erosion" "Solubility" _solubility_9_66a62d6f47fb09a92282f9998fa8e597_2 :!class /Script/Engine.KismetMaterialLibrary)
          (Rendering|Material|SetScalarParameterValue "/Landmass/Landscape/BlueprintBrushes/MPC/MPC_Erosion.MPC_Erosion" "Evaporation" _evaporation_11_c7bafcb548bbb82468aad2b764749543_2 :!class /Script/Engine.KismetMaterialLibrary)
          (Rendering|Material|SetScalarParameterValue "/Landmass/Landscape/BlueprintBrushes/MPC/MPC_Erosion.MPC_Erosion" "VelocityDamping" _velocitydamping_27_b1c0f90747c99e2d459d01982e31c616 :!class /Script/Engine.KismetMaterialLibrary)
          (Rendering|Material|SetScalarParameterValue "/Landmass/Landscape/BlueprintBrushes/MPC/MPC_Erosion.MPC_Erosion" "AdvectionErosion" _velocityadvection_16_390741a84fa9ab14da3866a3c931dbef :!class /Script/Engine.KismetMaterialLibrary)
          (Variables|Default|SetPosition%20Idx 0)
          (Variables|Default|SetVelocity%20Idx 0)
          (Debug|SpawnInstanceParticleMeshes _resolutiondivisor_30_9cec792142f360e868d2129d125b4a4d)
          ; Allocate Updated RTs for Downsampling
          (Utilities|FlowControl|Sequence
            (:then_0
              (Rendering|CreateAndAssignRenderTarget (Variables|Debug|RenderTargets|GetPosition%20RT%20A) _x_3 _y_3 "RTF_RGBA32f"
                (:Out
                  (Rendering|CreateAndAssignRenderTarget (Variables|Debug|RenderTargets|GetPosition%20RT%20B) _x_4 _y_4 "RTF_RGBA32f"))))
            (:then_1
              (Rendering|CreateAndAssignRenderTarget (Variables|Debug|RenderTargets|GetVelocity%20RT%20A) _x_5 _y_5 "RTF_RGBA32f"
                (:Out
                  (Rendering|CreateAndAssignRenderTarget (Variables|Debug|RenderTargets|GetVelocity%20RT%20B) _x_6 _y_6 "RTF_RGBA32f"
                    (:Out
                      (Rendering|CreateAndAssignRenderTarget (Variables|Debug|RenderTargets|GetSceneCapture%20RT) _x_2 _y_2 "RTF_RGBA32f"
                        (:Out
                          (Variables|/Script/Engine.SceneCaptureComponent2D|SetTextureTarget (Variables|Debug|RenderTargets|GetSceneCapture%20RT) (Variables|Default|GetScatterMeshSceneCapture))))))))))
          ; Clear and Seed RT and Assign MID
          (Rendering|ClearRenderTarget2D (Variables|Debug|RenderTargets|GetPosition%20RT%20A))
          (Rendering|ClearRenderTarget2D (Variables|Debug|RenderTargets|GetPosition%20RT%20B))
          (Rendering|ClearRenderTarget2D (Variables|Debug|RenderTargets|GetVelocity%20RT%20A) "(R=0.000000,G=0.000000,B=0.000000,A=0.000000)")
          (Rendering|ClearRenderTarget2D (Variables|Debug|RenderTargets|GetVelocity%20RT%20B) "(R=0.000000,G=0.000000,B=0.000000,A=0.000000)")
          (Rendering|Material|SetTextureParameterValue (Variables|Debug|MIDs|GetParticle%20Position%20Seed%20MID) "HeightRT" (Variables|Debug|RenderTargets|GetHeight%20RT%20A))
          (Rendering|DrawMaterialtoRenderTarget (Variables|Debug|RenderTargets|GetPosition%20RT%20A) (Variables|Debug|MIDs|GetParticle%20Position%20Seed%20MID))
          (Rendering|Material|SetTextureParameterValue (Variables|Debug|MIDs|GetParticle%20Preview%20MID) "PositionRT" (Variables|Debug|RenderTargets|GetPosition%20RT%20A))
          (Rendering|Material|SetTextureParameterValue (Variables|Debug|MIDs|GetParticle%20Weightmap%20MID) "SceneCaptureRT" (Variables|Debug|RenderTargets|GetSceneCapture%20RT))
          (Rendering|SceneCapture|CaptureScene (Variables|Default|GetScatterMeshSceneCapture) :!class /Script/Engine.SceneCaptureComponent2D) ; Capture to show initial particle state for Debug
          (Rendering|Material|SetScalarParameterValue "/Landmass/Landscape/BlueprintBrushes/MPC/MPC_Erosion.MPC_Erosion" "SimulatedResolutionX" (/ (Math|Conversions|ToFloat(Integer) _x) _resolutiondivisor_30_9cec792142f360e868d2129d125b4a4d_2) :!class /Script/Engine.KismetMaterialLibrary)
          (Rendering|Material|SetScalarParameterValue "/Landmass/Landscape/BlueprintBrushes/MPC/MPC_Erosion.MPC_Erosion" "SimulatedResolutionY" (/ (Math|Conversions|ToFloat(Integer) _y) _resolutiondivisor_30_9cec792142f360e868d2129d125b4a4d_2) :!class /Script/Engine.KismetMaterialLibrary)
          (Rendering|Material|SetScalarParameterValue "/Landmass/Landscape/BlueprintBrushes/MPC/MPC_Erosion.MPC_Erosion" "ParticleSize" (* (.x (.scale (Variables|Debug|GetLandscapeTransform))) _resolutiondivisor_30_9cec792142f360e868d2129d125b4a4d_2) :!class /Script/Engine.KismetMaterialLibrary)
          (for _array_element (Utilities|Struct|BreakLandmassParticleErosionSettings (Utilities|Struct|BreakLandmassErosionKeyframe (Variables|Debug|GetCurrentErosionKeyframe)))
            (Rendering|Material|SetScalarParameterValue _particle_weightmap_mid "Deposition" (.DepositionStrength_9_A0C69D6145F47973775021AB8749B742 (Utilities|Struct|BreakErodeWeightmap _array_element)) :!class /Script/Engine.MaterialInstanceDynamic)
            (Rendering|Material|SetScalarParameterValue _particle_weightmap_mid "Erosion" (.ErosionStrength_10_CF946B9A4ADA64BEEB4E8DA6A2CEA428 (Utilities|Struct|BreakErodeWeightmap _array_element)) :!class /Script/Engine.MaterialInstanceDynamic)
            (Rendering|Material|SetScalarParameterValue _particle_weightmap_mid "Velocity" (.VelocityStrength_14_FC9BD95E42791F6B55C07BAAF882F358 (Utilities|Struct|BreakErodeWeightmap _array_element)) :!class /Script/Engine.MaterialInstanceDynamic)))
        (:NewEnumerator3
          (Rendering|Material|SetScalarParameterValue "/Landmass/Landscape/BlueprintBrushes/MPC/MPC_Erosion.MPC_Erosion" "SlopeSmoothThreshold" (select (== _smoothslopesgreaterthan_18_b1c0f90747c99e2d459d01982e31c616 0.0) 999.0 _smoothslopesgreaterthan_18_b1c0f90747c99e2d459d01982e31c616) :!class /Script/Engine.KismetMaterialLibrary)
          (Rendering|Material|SetScalarParameterValue "/Landmass/Landscape/BlueprintBrushes/MPC/MPC_Erosion.MPC_Erosion" "FlatSmoothThreshold" (select (== _smoothslopeslessthan_20_ad0c525248c27f272b99899db760361a 0.0) -999.0 _smoothslopeslessthan_20_ad0c525248c27f272b99899db760361a) :!class /Script/Engine.KismetMaterialLibrary)
          (Rendering|Material|SetScalarParameterValue "/Landmass/Landscape/BlueprintBrushes/MPC/MPC_Erosion.MPC_Erosion" "SlopeSmoothingOpacity" _opacity_30_214597f847b9e91587d7c288f99ce3aa :!class /Script/Engine.KismetMaterialLibrary))
        (:NewEnumerator4
          (Rendering|Material|SetScalarParameterValue "/Landmass/Landscape/BlueprintBrushes/MPC/MPC_Erosion.MPC_Erosion" "StartWater" _startingwateramount_5_6662d7234e14e20ac900cbb1168ed16d :!class /Script/Engine.KismetMaterialLibrary)
          (Rendering|Material|SetScalarParameterValue "/Landmass/Landscape/BlueprintBrushes/MPC/MPC_Erosion.MPC_Erosion" "RainAmount" _rainperiteration_27_6fa2622d48a07f8ad76adc8a52d2335e :!class /Script/Engine.KismetMaterialLibrary)
          (Rendering|Material|SetScalarParameterValue "/Landmass/Landscape/BlueprintBrushes/MPC/MPC_Erosion.MPC_Erosion" "Solubility" _solubility_9_66a62d6f47fb09a92282f9998fa8e597 :!class /Script/Engine.KismetMaterialLibrary)
          (Rendering|Material|SetScalarParameterValue "/Landmass/Landscape/BlueprintBrushes/MPC/MPC_Erosion.MPC_Erosion" "Evaporation" _evaporation_11_c7bafcb548bbb82468aad2b764749543 :!class /Script/Engine.KismetMaterialLibrary)
          ;[1 1 1] Seed Start Water
          (Rendering|Material|SetTextureParameterValue _water_levelling_seed_mid "HeightRT" _read)
          (Rendering|DrawMaterialtoRenderTarget _write _water_levelling_seed_mid)
          (Math|Integer|IncrementInt (Variables|Default|GetHeight%20idx)
            (:%20%20
              (for _array_element_2 (Utilities|Struct|BreakLandmassWaterLevellingSettings (Utilities|Struct|BreakLandmassErosionKeyframe (Variables|Debug|GetCurrentErosionKeyframe)))
                (Rendering|Material|SetScalarParameterValue _water_levelling_weightmap_mid "Deposition" (.DepositionStrength_9_A0C69D6145F47973775021AB8749B742 (Utilities|Struct|BreakErodeWeightmap _array_element_2)) :!class /Script/Engine.MaterialInstanceDynamic)
                (Rendering|Material|SetScalarParameterValue _water_levelling_weightmap_mid "Erosion" (.ErosionStrength_10_CF946B9A4ADA64BEEB4E8DA6A2CEA428 (Utilities|Struct|BreakErodeWeightmap _array_element_2)) :!class /Script/Engine.MaterialInstanceDynamic)))))
        (:NewEnumerator1
          (Rendering|Material|SetScalarParameterValue (Variables|Debug|MIDs|GetAdd%20Noise%20MID) "Noise Height" _noiseheightmeters_20_74ba1f384b27ba15040a558b6d44a4e3 :!class /Script/Engine.MaterialInstanceDynamic)
          (Rendering|Material|SetScalarParameterValue (Variables|Debug|MIDs|GetAdd%20Noise%20MID) "Noise Scale" _noisescalemeters_21_c8ee95774aa7dca911e3f59e537a0cbb :!class /Script/Engine.MaterialInstanceDynamic)
          (Rendering|Material|SetScalarParameterValue (Variables|Debug|MIDs|GetAdd%20Noise%20MID) "Noise Octaves" _noiseoctaves_4_d613165642c15bc55548b69690950d03 :!class /Script/Engine.MaterialInstanceDynamic)
          (Rendering|Material|SetScalarParameterValue (Variables|Debug|MIDs|GetAdd%20Noise%20MID) "Noise Persistence" _noisepersistence_11_fa6ffe4e40449bf51d5ddf9c79362a04 :!class /Script/Engine.MaterialInstanceDynamic)
          (Rendering|Material|SetScalarParameterValue (Variables|Debug|MIDs|GetAdd%20Noise%20MID) "Noise Seed" _noiseseed_14_ca1226e144c86e22d65c449b9fff7c91 :!class /Script/Engine.MaterialInstanceDynamic)
          (Rendering|Material|SetScalarParameterValue (Variables|Debug|MIDs|GetAdd%20Noise%20MID) "Turbulence" (Math|Conversions|ToFloat(Boolean) _turbulence_16_f6422936458b63815688e498d8f934d5) :!class /Script/Engine.MaterialInstanceDynamic)
          (Rendering|Material|SetScalarParameterValue (Variables|Debug|MIDs|GetAdd%20Noise%20MID) "Noise Floor" _noisefloor0_1_22_6deae6d549095a9ca2c08daab0f22af5 :!class /Script/Engine.MaterialInstanceDynamic)
          (Rendering|Material|SetTextureParameterValue _add_noise_mid "NormalRT" (Variables|Debug|RenderTargets|GetNormal%20RT))
          (Utilities|SetSlopeMaskParams _add_noise_mid _slopemask_25_f63fef4841c61d02307971af848d0f4b))
        (:NewEnumerator2
          (Rendering|CreateAndAssignMID (Variables|Debug|MIDs|GetCustom%20Material%20Erosion%20MID) _material_19_9623e74447d05ea3c529e584a7638d11 "Custom_ErosionMaterial_MID"
            (:Out
              (Initialization|SetCustomMIDParams (Variables|Debug|MIDs|GetCustom%20Material%20Erosion%20MID) _scalarparameters_24_30775f0f4d9c8b634089f9b367097495 _vectorparameters_36_cefaeee94c49cc8b1958458614c3c8cc _textureparameters_33_1825952e489caf47f454d089a554972f))))))
    (:then_1
      (Utilities|ComputeNormal)
      (Utilities|GetIterationCount)
      (Development|PrintString (Utilities|String|Append "Init erosion Mode: " (Utilities|String|EnumtoString (Utilities|Struct|BreakLandmassErosionKeyframe (Variables|Debug|GetCurrentErosionKeyframe))) " and particles visible = " (Utilities|String|ToString(Boolean) (and (Utilities|Enum|Equal(Enum) (Utilities|Struct|BreakLandmassErosionKeyframe (Variables|Debug|GetCurrentErosionKeyframe)) "NewEnumerator0") (Variables|Debug|GetShow%20Water)))) false))))

(fn InitializeMIDs
  (bind _landscape_mesh_preview_mid (Variables|Debug|MIDs|GetLandscape%20Mesh%20Preview%20MID))
  (Utilities|FlowControl|Sequence
    (:then_0
      ;[1 1 1] Create Particle Sim MIDs
      (Rendering|CreateAndAssignMID (Variables|Debug|MIDs|GetParticle%20Position%20Seed%20MID) "/Landmass/Landscape/ErosionBrush/Materials/Particles/ParticleErosion_WaterSeed.ParticleErosion_WaterSeed" "Particle_Seed_MID"
        (:Out
          (Rendering|CreateAndAssignMID (Variables|Debug|MIDs|GetScatter%20Position%20MID) "/Landmass/Landscape/ErosionBrush/Materials/Particles/ParticleErosion_ScatterPosition.ParticleErosion_ScatterPosition" "Particle_ScatterPosition_MID"
            (:Out
              (Rendering|CreateAndAssignMID (Variables|Debug|MIDs|GetScatter%20Velocity%20MID) "/Landmass/Landscape/ErosionBrush/Materials/Particles/ParticleErosion_ScatterVelocity.ParticleErosion_ScatterVelocity" "Particle_ScatterVelocity_MID"
                (:Out
                  (Rendering|CreateAndAssignMID (Variables|Debug|MIDs|GetScatter%20Height%20Erosion%20MID) "/Landmass/Landscape/ErosionBrush/Materials/Particles/ParticleErosion_ErodeTerrain.ParticleErosion_ErodeTerrain" "Particle_ErodeTerrain_MID"
                    (:Out
                      (Rendering|CreateAndAssignMID (Variables|Debug|MIDs|GetParticle%20Weightmap%20MID) "/Landmass/Landscape/ErosionBrush/Materials/Particles/ParticleErosion_Weightmap.ParticleErosion_Weightmap" "Particle_ErodeTerrain_Weightmap_MID"
                        (:Out
                          (Rendering|CreateAndAssignMID (Variables|Debug|MIDs|GetParticle%20Preview%20MID) "/Landmass/Landscape/ErosionBrush/Materials/Particles/ParticleErosion_ParticlePreview.ParticleErosion_ParticlePreview" "Particle_Preview_MID"
                            (:Out
                              (Rendering|CreateAndAssignMID (Variables|Debug|MIDs|GetParticle%20Render%20MID) (Variables|Debug|GetParticleRenderMat) "Particle_Render_MID"
                                (:Out
                                  (Variables|Debug|MIDs|SetParticle%20Sim%20MIDs (Utilities|Array|MakeArray (Variables|Debug|MIDs|GetParticle%20Position%20Seed%20MID) (Variables|Debug|MIDs|GetScatter%20Position%20MID) (Variables|Debug|MIDs|GetScatter%20Velocity%20MID) (Variables|Debug|MIDs|GetScatter%20Height%20Erosion%20MID) (Variables|Debug|MIDs|GetParticle%20Preview%20MID) (Variables|Debug|MIDs|GetParticle%20Render%20MID)))
                                  (Rendering|Material|SetMaterial (Variables|Default|GetScatter%20Particles%20ISMC) 0 (select (Variables|Debug|GetDisplay%20Particle%20Render%20Mat) (Variables|Debug|MIDs|GetParticle%20Render%20MID) (Variables|Debug|MIDs|GetParticle%20Preview%20MID)))
                                  ;[0.1067 0.5288 1] Assign Normal RT to Particle Sim MIDs
                                  (for _array_element (Variables|Debug|MIDs|GetParticle%20Sim%20MIDs)
                                    (Rendering|Material|SetTextureParameterValue _array_element "NormalRT" (Variables|Debug|RenderTargets|GetNormal%20RT))
                                    (Rendering|Material|SetTextureParameterValue (Variables|Debug|MIDs|GetScatter%20Height%20Erosion%20MID) "SceneCaptureRT" (Variables|Debug|RenderTargets|GetSceneCapture%20RT))))))))))))))))))
    (:then_1
      ;[1 1 1] Create Landscape Preview MID
      (Rendering|CreateAndAssignMID (Variables|Debug|MIDs|GetLandscape%20Mesh%20Preview%20MID) "/Landmass/Landscape/ErosionBrush/Materials/Preview/ErosionBrush_LandscapeMeshPreview.ErosionBrush_LandscapeMeshPreview" "Landscape_Mesh_Preview_MID"
        (:Out
          (Rendering|Material|SetTextureParameterValue _landscape_mesh_preview_mid "HeightRT" (Variables|Debug|RenderTargets|GetHeight%20RT%20A))
          (Rendering|Material|SetTextureParameterValue _landscape_mesh_preview_mid "NormalRT" (Variables|Debug|RenderTargets|GetNormal%20RT))
          (Rendering|Material|SetTextureParameterValue _landscape_mesh_preview_mid "SceneCaptureRT" (Variables|Debug|RenderTargets|GetSceneCapture%20RT))
          (Rendering|Material|SetMaterial (Variables|Default|GetLandscape%20ISMC) 0 (Variables|Debug|MIDs|GetLandscape%20Mesh%20Preview%20MID))
          (bind _keys (Utilities|Map|Keys (Variables|Debug|RenderTargets|GetWeightmapRTs)))
          (for (_array_element_2 _array_index) _keys
            (Rendering|Material|SetTextureParameterValue (Variables|Debug|MIDs|GetLandscape%20Mesh%20Preview%20MID) (Utilities|String|StringToName (Utilities|String|Append "Weightmap" (Utilities|String|ToString(Integer) _array_index))) (Utilities|Map|Find (Variables|Debug|RenderTargets|GetWeightmapRTs) _array_element_2))))))
    (:then_2
      ;[1 1 1] Create Pack and Unpack MIDs
      (Rendering|CreateAndAssignMID (Variables|Debug|MIDs|GetUnpack%20RG8%20Height%20MID) "/Landmass/Landscape/ErosionBrush/Materials/Packing/ErosionBrush_UnpackRG8Height.ErosionBrush_UnpackRG8Height" "Erosion_UnpackRG8_MID"
        (:Out
          (Rendering|CreateAndAssignMID (Variables|Debug|MIDs|GetMake%20Erosion%20Relative%20RG8%20Height%20MID) "/Landmass/Landscape/ErosionBrush/Materials/Packing/ErosionBrush_PackErosionRelative.ErosionBrush_PackErosionRelative" "Erosion_MakeRelative_MID"
            (:Out
              ;[1 1 1] Create Water Levelling MID
              (Rendering|CreateAndAssignMID (Variables|Debug|MIDs|GetWater%20Levelling%20MID) "/Landmass/Landscape/ErosionBrush/Materials/WaterLevelling/Erosion_WaterLevelling.Erosion_WaterLevelling" "Erosion_WaterLevelling_MID"
                (:Out
                  (Rendering|CreateAndAssignMID (Variables|Debug|MIDs|GetWater%20Levelling%20Weightmap%20MID) "/Landmass/Landscape/ErosionBrush/Materials/WaterLevelling/Erosion_WaterLevelling_Weightmap.Erosion_WaterLevelling_Weightmap" "Erosion_WaterLevelling_Weightmap_MID"
                    (:Out
                      (Rendering|CreateAndAssignMID (Variables|Debug|MIDs|GetWater%20Levelling%20Seed%20MID) "/Landmass/Landscape/ErosionBrush/Materials/WaterLevelling/Erosion_WaterLevelling_SeedWater.Erosion_WaterLevelling_SeedWater" "Erosion_WaterLevelling_Weightmap_MID"))))))))))
    (:then_4
      ;[1 1 1] Create Slope Smoothing MID
      (Rendering|CreateAndAssignMID (Variables|Debug|MIDs|GetSlope%20Smoothing%20MID) "/Landmass/Landscape/ErosionBrush/Materials/SlopeSmoothing/Erosion_SlopeSmoothing.Erosion_SlopeSmoothing" "Erosion_SlopeSmoothing_MID"
        (:Out
          ;[1 1 1] Create Add Noise MID
          (Rendering|CreateAndAssignMID (Variables|Debug|MIDs|GetAdd%20Noise%20MID) "/Landmass/Landscape/ErosionBrush/Materials/AddNoise/ErosionBrush_AddNoise.ErosionBrush_AddNoise" "Erosion_SlopeSmoothing_MID"))))
    (:then_6
      ;[0.1146 0.4559 1] Create Compute Normals MID
      (Rendering|CreateAndAssignMID (Variables|Debug|MIDs|GetCompute%20Normals%20MID) "/Landmass/Landscape/ErosionBrush/Materials/Normals/Erosion_ComputeNormals.Erosion_ComputeNormals" "Erosion_ComputeNormals_MID"))))

(fn InitializeRTs
  (bind (_x _y) (Utilities|Struct|BreakIntPoint (Variables|Debug|GetRT%20Resolution)))
  (bind (_x_2 _y_2) (Utilities|Struct|BreakIntPoint (Variables|Debug|GetRT%20Resolution)))
  (bind (_x_3 _y_3) (Utilities|Struct|BreakIntPoint (Variables|Debug|GetRT%20Resolution)))
  (bind (_x_4 _y_4) (Utilities|Struct|BreakIntPoint (Variables|Debug|GetRT%20Resolution)))
  (Utilities|FlowControl|Sequence
    (:then_0
      (Rendering|CreateAndAssignRenderTarget (Variables|Debug|RenderTargets|GetHeight%20RT%20A) _x_2 _y_2 "RTF_RG32f"
        (:Out
          (Rendering|CreateAndAssignRenderTarget (Variables|Debug|RenderTargets|GetHeight%20RT%20B) _x_4 _y_4 "RTF_RG32f"))))
    (:then_3
      (Rendering|CreateAndAssignRenderTarget (Variables|Debug|RenderTargets|GetNormal%20RT) _x _y "RTF_RGBA32f"
        (:Out
          (Rendering|CreateAndAssignRenderTarget (Variables|Debug|RenderTargets|GetErosion_Height_RT_Data) _x_3 _y_3 "RTF_R16f"))))))

(fn InitializeSimulation
  (bind _landscapetransform (Variables|Debug|GetLandscapeTransform))
  (bind _scale (.scale _landscapetransform))
  (bind _returnvalue (Math|Conversions|ToVector2D(Vector) _scale))
  (Initialization|InitializeSize)
  (Initialization|InitializeRTs)
  (Initialization|InitializeMIDs)
  (Initialization|SetMPCValues)
  (Variables|/Script/Engine.SceneCaptureComponent2D|SetTextureTarget (Variables|Debug|RenderTargets|GetSceneCapture%20RT) (Variables|Default|GetScatterMeshSceneCapture))
  (Rendering|SceneCapture|ShowOnlyComponent (Variables|Default|GetScatterMeshSceneCapture) (Variables|Default|GetScatter%20Particles%20ISMC))
  (Variables|/Script/Engine.SceneCaptureComponent2D|SetOrthoWidth (Math|Vector2D|BreakVector2D (* (Math|Conversions|ToVector2D(IntPoint) (Variables|Debug|GetRT%20Resolution)) _returnvalue)) (Variables|Default|GetScatterMeshSceneCapture))
  (Transformation|SetWorldLocation (Variables|Default|GetScatterMeshSceneCapture) (Math|Conversions|ToVector(Vector2D) (- (Variables|Debug|GetRT%20World%20Center) (* _returnvalue "(X=0.500000,Y=0.500000)")) (+ (.z (.location (Variables|Debug|GetLandscapeTransform))) 5e+04)) false false)
  (Rendering|Material|SetTextureParameterValue (Variables|Debug|MIDs|GetUnpack%20RG8%20Height%20MID) "HeightRT" (Variables|Debug|RenderTargets|GetLandscape%20Height%20Copy%20RT))
  (Rendering|DrawMaterialtoRenderTarget (Variables|Debug|RenderTargets|GetHeight%20RT%20A) (Variables|Debug|MIDs|GetUnpack%20RG8%20Height%20MID)))

(fn InitializeSize
  (bind (_x _y) (Utilities|Struct|BreakIntPoint (Variables|Debug|GetLandscape%20Size)))
  (bind _landscape_size (Variables|Debug|GetLandscape%20Size))
  (bind (_x_2 _y_2) (Utilities|Struct|BreakIntPoint _landscape_size))
  (Variables|Debug|SetWorldSize (Math|Vector2D|MakeVector2D (* (.x (.scale (Variables|Debug|GetLandscapeTransform))) _x) (* (.y (.scale (Variables|Debug|GetLandscapeTransform))) _y)))
  (Variables|Debug|SetRT%20World%20Center (+ (Math|Conversions|ToVector2D(Vector) (.location (Variables|Debug|GetLandscapeTransform))) (* (* (Math|Conversions|ToVector2D(Vector) (.scale (Variables|Debug|GetLandscapeTransform))) (Variables|Debug|GetRT%20Resolution)) 0.5)))
  (Variables|Debug|SetSections%20X (/ _x_2 255))
  (Variables|Debug|SetSections%20Y (/ _y_2 255)))

(fn LaunchErosionTimeline
  (bind _returnvalue (Game|SpawnActorfromClass "/Landmass/Landscape/ErosionBrush/BP/Widgets/EUB_SpawnErosionTab.EUB_SpawnErosionTab_C" (Math|Conversions|ToTransform(Vector) (Math|Vector|MakeVector 0.0)) "Undefined" "MultiplyWithRoot" 0 0))
  (Class|EUBSpawnErosionTab|SpawnErosionWidget _returnvalue self)
  (Actor|DestroyActor _returnvalue)
  (Landscape|RequestLandscapeUpdate true)
  (bind _outactors (Actor|GetAllActorsOfClass "/Landmass/Landscape/BlueprintBrushes/LandmassBrushManager.LandmassBrushManager_C"))
  (Default|TogglePreviewMode (Utilities|Array|Get(acopy) _outactors) false))

(fn Pause
  (Variables|Default|SetSimulating (not (Variables|Default|GetSimulating))))

(fn RenderErosionLayer
  (params (In%20Parameters :type Landscape%20Brush%20Parameters%20Structure) (Extents :type Vector%204%20Structure) (BrushManager :type Landmass%20Brush%20Manager%20Object%20Reference))
  (bind (_renderareaworldtransform _renderareasize _combinedresult _layertype) (Utilities|Struct|BreakLandscapeBrushParameters In%20Parameters))
  (bind (_renderareaworldtransform_2 _renderareasize_2 _weightmaplayername) (Utilities|Struct|BreakLandscapeBrushParameters In%20Parameters))
  (bind _erosionweightmaptexturessaved (Variables|Debug|GetErosionWeightmapTexturesSaved))
  (bind (_value _returnvalue) (Utilities|Map|Find _erosionweightmaptexturessaved _weightmaplayername))
  (bind _apply_erosion_to_landscape_mid (Variables|Debug|MIDs|GetApply%20Erosion%20to%20Landscape%20MID))
  (Development|PrintString "ErosionBrush Preview Render Requested" false)
  (Variables|Debug|SetHeightmapRender (Utilities|Enum|Equal(Enum) _layertype "Heightmap"))
  (Variables|Debug|RenderTargets|SetLandscape%20In%20RT _combinedresult)
  (Initialization|InitializeBrush)
  (if (Variables|Debug|GetHeightmapRender)
    ;[1 1 1] Initially, just Cache the Landscape height for use in Simulation
    (Rendering|Material|SetTextureParameterValue (Variables|Debug|MIDs|GetCopy%20RT%20MID) "RT" (Variables|Debug|RenderTargets|GetLandscape%20In%20RT))
    (Rendering|DrawMaterialtoRenderTarget (Variables|Debug|RenderTargets|GetLandscape%20Height%20Copy%20RT) (Variables|Debug|MIDs|GetCopy%20RT%20MID))
    (else
      _))
  ;[1 0.04654 0] Check if Simulation is Complete
  (if (Variables|Debug|GetIs%20Simulation%20Done)
    (if (Variables|Debug|GetHeightmapRender)
      (Utilities|IsValid (Variables|Debug|GetErosion_Height_TextureSaved)
        (:Is%20Valid
          ;[0.1476 0.5379 1] Apply Saved Erosion Heightmap
          (Rendering|Material|SetTextureParameterValue _apply_erosion_to_landscape_mid "LandscapeRT" (Variables|Debug|RenderTargets|GetLandscape%20Height%20Copy%20RT))
          (Rendering|Material|SetTextureParameterValue _apply_erosion_to_landscape_mid "Erosion" (Variables|Debug|GetErosion_Height_TextureSaved))
          (Utilities|BrushDrawToRenderTargetExtents (Variables|Debug|RenderTargets|GetLandscape%20In%20RT) (Variables|Debug|MIDs|GetApply%20Erosion%20to%20Landscape%20MID) Extents (Variables|/Script/LandmassEditor.LandmassManagerBase|GetLandscapeInformation BrushManager)) ; Render Brush into the Original Source\nThis version draws to the entire current brushes bounds rather than restricting to the edited brush bounds\nIt doesn't know about which brush was edited as the call comes from the Landscape itself.
        )
        (:Is%20Not%20Valid
          (Development|PrintString "Erosion Brush: Erosion Not rendered because Erosion Texture Data was Null." false)))
      (else
        ;[1 0.697 0.1264] Apply Erosion Weightmap
        (if _returnvalue
          (Utilities|IsValid _value
            (:Is%20Valid
              (Rendering|Material|SetTextureParameterValue (Variables|Debug|MIDs|GetApply%20Weightmap%20MID) "WeightmapRT" _value)
              (Utilities|BrushDrawToRenderTargetExtents (Variables|Debug|RenderTargets|GetLandscape%20In%20RT) (Variables|Debug|MIDs|GetApply%20Weightmap%20MID) Extents (Variables|/Script/LandmassEditor.LandmassManagerBase|GetLandscapeInformation BrushManager)) ; Render Brush into the Original Source\nThis version draws to the entire current brushes bounds rather than restricting to the edited brush bounds\nIt doesn't know about which brush was edited as the call comes from the Landscape itself.
            )
            (:Is%20Not%20Valid
              (Development|PrintString "Erosion Brush: Erosion Not rendered because Erosion Texture Data was Null." false))))))
    ;[0.1699 0.01189 0.25] Return the result to Landscape
    (return false)))

(fn RenderLayer
  (params (InParameters :type Landscape%20Brush%20Parameters%20Structure%20%28by%20ref%29))
  (bind (_renderareaworldtransform _renderareasize _combinedresult _layertype) (Utilities|Struct|BreakLandscapeBrushParameters InParameters))
  (bind _inparameters (Variables|GetInParameters))
  (bind (_renderareaworldtransform_2 _renderareasize_2 _weightmaplayername) (Utilities|Struct|BreakLandscapeBrushParameters _inparameters))
  (bind _erosionweightmaptexturessaved (Variables|Debug|GetErosionWeightmapTexturesSaved))
  (bind (_value _returnvalue) (Utilities|Map|Find _erosionweightmaptexturessaved _weightmaplayername))
  (bind _apply_erosion_to_landscape_mid (Variables|Debug|MIDs|GetApply%20Erosion%20to%20Landscape%20MID))
  (Variables|Debug|SetHeightmapRender (Utilities|Enum|Equal(Enum) _layertype "Heightmap"))
  (Variables|Debug|RenderTargets|SetLandscape%20In%20RT _combinedresult)
  (Initialization|InitializeBrush)
  (if (Variables|Debug|GetHeightmapRender)
    ;[1 1 1] Initially, just Cache the Landscape height for use in Simulation
    (Rendering|Material|SetTextureParameterValue (Variables|Debug|MIDs|GetCopy%20RT%20MID) "RT" (Variables|Debug|RenderTargets|GetLandscape%20In%20RT))
    (Rendering|DrawMaterialtoRenderTarget (Variables|Debug|RenderTargets|GetLandscape%20Height%20Copy%20RT) (Variables|Debug|MIDs|GetCopy%20RT%20MID))
    (else
      _))
  ;[1 0.04654 0] Check if Simulation is Complete
  (if (Variables|Debug|GetIs%20Simulation%20Done)
    (if (Variables|Debug|GetHeightmapRender)
      (Utilities|IsValid (Variables|Debug|GetErosion_Height_TextureSaved)
        (:Is%20Valid
          ;[0.1476 0.5379 1] Apply Saved Erosion Heightmap
          (Rendering|Material|SetTextureParameterValue _apply_erosion_to_landscape_mid "LandscapeRT" (Variables|Debug|RenderTargets|GetLandscape%20Height%20Copy%20RT))
          (Rendering|Material|SetTextureParameterValue _apply_erosion_to_landscape_mid "Erosion" (Variables|Debug|GetErosion_Height_TextureSaved))
          (Rendering|DrawMaterialtoRenderTarget (Variables|Debug|RenderTargets|GetLandscape%20In%20RT) (Variables|Debug|MIDs|GetApply%20Erosion%20to%20Landscape%20MID)))
        (:Is%20Not%20Valid
          (Development|PrintString "Erosion Brush: Erosion Not rendered because Erosion Texture Data was Null." false)))
      (else
        ;[1 0.697 0.1264] Apply Erosion Weightmap
        (if _returnvalue
          (Utilities|IsValid _value
            (:Is%20Valid
              (Rendering|Material|SetTextureParameterValue (Variables|Debug|MIDs|GetApply%20Weightmap%20MID) "WeightmapRT" _value)
              (Rendering|DrawMaterialtoRenderTarget (Variables|Debug|RenderTargets|GetLandscape%20In%20RT) (Variables|Debug|MIDs|GetApply%20Weightmap%20MID)))
            (:Is%20Not%20Valid
              (Development|PrintString "Erosion Brush: Erosion Not rendered because Erosion Texture Data was Null." false))))))
    ;[0.1699 0.01189 0.25] Return the result to Landscape
    (Default|CallErosionBrushUpdated)
    (return (Variables|Debug|RenderTargets|GetLandscape%20In%20RT))))

(fn Reset
  (Variables|Default|SetCurrent%20Iteration 0)
  (Variables|Default|SetHeight%20idx 0)
  (Variables|Default|SetPosition%20Idx 0)
  (Variables|Default|SetVelocity%20Idx 0)
  (Variables|Default|SetIteration%20Key%20Index 0)
  (Rendering|ClearRenderTarget2D (Variables|Debug|RenderTargets|GetPosition%20RT%20A))
  (Rendering|ClearRenderTarget2D (Variables|Debug|RenderTargets|GetPosition%20RT%20B))
  (Rendering|ClearRenderTarget2D (Variables|Debug|RenderTargets|GetVelocity%20RT%20A) "(R=0.000000,G=0.000000,B=0.000000,A=0.000000)")
  (Rendering|ClearRenderTarget2D (Variables|Debug|RenderTargets|GetVelocity%20RT%20B) "(R=0.000000,G=0.000000,B=0.000000,A=0.000000)")
  (Utilities|FlowControl|Sequence
    (:then_0
      (Initialization|InitializeSimulation)
      (Initialization|InitializeErosionValues (Utilities|Array|Get(aref) (Variables|ErosionSettings|GetErosion%20Keyframes)))
      (Variables|Debug|SetIs%20Simulation%20Done false)
      (Variables|Default|SetSimulating false)
      (Rendering|SetVisibility (Variables|Default|GetScatter%20Particles%20ISMC) false false)
      (Landscape|RequestLandscapeUpdate true))
    (:then_1
      (bind _keys (Utilities|Map|Keys (Variables|Debug|RenderTargets|GetWeightmapRTs)))
      (for _array_element _keys
        (Rendering|ClearRenderTarget2D (Utilities|Map|Find (Variables|Debug|RenderTargets|GetWeightmapRTs) _array_element))))))

(fn SaveResultasRelativeTexture
  (bind (_seconds _partialseconds) (Utilities|Time|GetAccurateRealTime))
  (bind _weightmaprts (Variables|Debug|RenderTargets|GetWeightmapRTs))
  (bind _temptexture (LocalVariables|GetTempTexture))
  (bind _erosion_height_texturesaved (Variables|Debug|GetErosion_Height_TextureSaved))
  (bind _erosionweightmaptexturessaved (Variables|Debug|GetErosionWeightmapTexturesSaved))
  (bind _make_erosion_relative_rg8_height_mid (Variables|Debug|MIDs|GetMake%20Erosion%20Relative%20RG8%20Height%20MID))
  (Utilities|FlowControl|Sequence
    (:then_0
      (Rendering|Material|SetTextureParameterValue _make_erosion_relative_rg8_height_mid "LandscapeRT" (Variables|Debug|RenderTargets|GetLandscape%20Height%20Copy%20RT))
      (Rendering|Material|SetTextureParameterValue _make_erosion_relative_rg8_height_mid "HeightRT" (Utilities|HeightPingPong))
      (Rendering|DrawMaterialtoRenderTarget (Variables|Debug|RenderTargets|GetErosion_Height_RT_Data) (Variables|Debug|MIDs|GetMake%20Erosion%20Relative%20RG8%20Height%20MID))
      (Utilities|IsValid (Variables|Debug|GetErosion_Height_TextureSaved)
        (:Is%20Valid
          _)
        (:Is%20Not%20Valid
          (bind _returnvalue (Game|ConstructObjectfromClass "/Script/Engine.Texture2D"))
          (Variables|Debug|SetErosion_Height_TextureSaved _returnvalue)))
      (Variables|/Script/Engine.Texture|SetMipGenSettings "TMGS_NoMipmaps" _erosion_height_texturesaved)
      (EditorScripting|Rendering|ConvertRenderTargettoTexture2DEditorOnly (Variables|Debug|RenderTargets|GetErosion_Height_RT_Data) _erosion_height_texturesaved)
      (Variables|/Script/Engine.Texture|SetCompressionSettings "TC_HalfFloat" (Variables|Debug|GetErosion_Height_TextureSaved))
      (Development|PrintString (Utilities|String|Append "Texture saving done in: " (Utilities|String|ToString(Float) (- (+ (Math|Conversions|ToFloat(Integer) _seconds) _partialseconds) (Variables|Default|GetEnd%20Time))) " seconds")))
    (:then_1
      (bind _keys (Utilities|Map|Keys (Variables|Debug|RenderTargets|GetWeightmapRTs)))
      (for _array_element _keys
        (if (.ReturnValue (Utilities|Map|Find _weightmaprts _array_element))
          (LocalVariables|SetWeightmapRT (Utilities|Map|Find _weightmaprts _array_element))
          (if (.ReturnValue (Utilities|Map|Find _erosionweightmaptexturessaved _array_element))
            (bind _output_get (LocalVariables|SetTempTexture (Utilities|Map|Find _erosionweightmaptexturessaved _array_element)))
            (Utilities|IsValid _output_get
              (:Is%20Valid
                _)
              (:Is%20Not%20Valid
                (bind _returnvalue_2 (Game|ConstructObjectfromClass "/Script/Engine.Texture2D"))
                (LocalVariables|SetTempTexture _returnvalue_2)))
            (Variables|/Script/Engine.Texture|SetMipGenSettings "TMGS_NoMipmaps" _temptexture)
            (EditorScripting|Rendering|ConvertRenderTargettoTexture2DEditorOnly (LocalVariables|GetWeightmapRT) _temptexture)
            (Utilities|Map|Add (Variables|Debug|GetErosionWeightmapTexturesSaved) _array_element (LocalVariables|GetTempTexture))
            (else
              _)))))))

(fn SetCustomMIDParams
  (params (In%20MID :type Material%20Instance%20Dynamic%20Object%20Reference) (In%20Floats :type Map%20of%20Names%20to%20Float%20%28double-precision%29s) (In%20Vectors :type Map%20of%20Names%20to%20Linear%20Color%20Structures) (In%20Textures :type Map%20of%20Names%20to%20Texture%202D%20Object%20References))
  (Utilities|FlowControl|Sequence
    (:then_0
      (bind _keys (Utilities|Map|Keys (Variables|GetIn%20Floats)))
      (for _array_element _keys
        (Rendering|Material|SetScalarParameterValue (Variables|GetIn%20MID) _array_element (Utilities|Map|Find (Variables|GetIn%20Floats) _array_element) :!class /Script/Engine.MaterialInstanceDynamic)))
    (:then_1
      (bind _keys_2 (Utilities|Map|Keys (Variables|GetIn%20Vectors)))
      (for _array_element_2 _keys_2
        (Rendering|Material|SetVectorParameterValue (Variables|GetIn%20MID) _array_element_2 (Utilities|Map|Find (Variables|GetIn%20Vectors) _array_element_2) :!class /Script/Engine.MaterialInstanceDynamic)))
    (:then_2
      (bind _keys_3 (Utilities|Map|Keys (Variables|GetIn%20Textures)))
      (for _array_element_3 _keys_3
        (Rendering|Material|SetTextureParameterValue (Variables|GetIn%20MID) _array_element_3 (Utilities|Map|Find (Variables|GetIn%20Textures) _array_element_3))))))

(fn SetMPCValues
  (bind _rt_resolution (Variables|Debug|GetRT%20Resolution))
  (bind (_x _y) (Utilities|Struct|BreakIntPoint _rt_resolution))
  (bind _rt_resolution_2 (Variables|Debug|GetRT%20Resolution))
  (bind (_x_2 _y_2) (Utilities|Struct|BreakIntPoint _rt_resolution_2))
  (bind _rt_resolution_3 (Variables|Debug|GetRT%20Resolution))
  (bind (_x_3 _y_3) (Utilities|Struct|BreakIntPoint _rt_resolution_3))
  (bind _worldsize (Variables|Debug|GetWorldSize))
  (bind (_x_4 _y_4) (Math|Vector2D|BreakVector2D _worldsize))
  (bind _landscapetransform (Variables|Debug|GetLandscapeTransform))
  (bind _scale (.scale _landscapetransform))
  (bind _landscapetransform_2 (Variables|Debug|GetLandscapeTransform))
  (bind _scale_2 (.scale _landscapetransform_2))
  (Rendering|Material|SetScalarParameterValue "/Landmass/Landscape/BlueprintBrushes/MPC/MPC_Erosion.MPC_Erosion" "TotalSizeX" _x_4 :!class /Script/Engine.KismetMaterialLibrary)
  (Rendering|Material|SetScalarParameterValue "/Landmass/Landscape/BlueprintBrushes/MPC/MPC_Erosion.MPC_Erosion" "TotalSizeY" _y_4 :!class /Script/Engine.KismetMaterialLibrary)
  (Rendering|Material|SetScalarParameterValue "/Landmass/Landscape/BlueprintBrushes/MPC/MPC_Erosion.MPC_Erosion" "ResX" (Math|Conversions|ToFloat(Integer) _x_3) :!class /Script/Engine.KismetMaterialLibrary)
  (Rendering|Material|SetScalarParameterValue "/Landmass/Landscape/BlueprintBrushes/MPC/MPC_Erosion.MPC_Erosion" "ResY" (Math|Conversions|ToFloat(Integer) _y_3) :!class /Script/Engine.KismetMaterialLibrary)
  (Rendering|Material|SetScalarParameterValue "/Landmass/Landscape/BlueprintBrushes/MPC/MPC_Erosion.MPC_Erosion" "RTWorldSizeX" (* (Math|Conversions|ToFloat(Integer) _x) (.x _scale_2)) :!class /Script/Engine.KismetMaterialLibrary)
  (Rendering|Material|SetScalarParameterValue "/Landmass/Landscape/BlueprintBrushes/MPC/MPC_Erosion.MPC_Erosion" "RTWorldSizeY" (* (Math|Conversions|ToFloat(Integer) _y) (.y _scale_2)) :!class /Script/Engine.KismetMaterialLibrary)
  (Rendering|Material|SetScalarParameterValue "/Landmass/Landscape/BlueprintBrushes/MPC/MPC_Erosion.MPC_Erosion" "GridSize" (.x _scale) :!class /Script/Engine.KismetMaterialLibrary)
  (Rendering|Material|SetScalarParameterValue "/Landmass/Landscape/BlueprintBrushes/MPC/MPC_Erosion.MPC_Erosion" "ZScale" (/ (.z _scale) 128.0) :!class /Script/Engine.KismetMaterialLibrary)
  (Rendering|Material|SetScalarParameterValue "/Landmass/Landscape/BlueprintBrushes/MPC/MPC_Erosion.MPC_Erosion" "ComponentRes" (Math|Conversions|ToFloat(Integer) (Variables|Debug|GetComponentRes)) :!class /Script/Engine.KismetMaterialLibrary)
  (Rendering|Material|SetVectorParameterValue "/Landmass/Landscape/BlueprintBrushes/MPC/MPC_Erosion.MPC_Erosion" "LandscapeLocation" (Math|Conversions|ToLinearColor(Vector) (.location (Variables|Debug|GetLandscapeTransform))) :!class /Script/Engine.KismetMaterialLibrary)
  (Rendering|Material|SetScalarParameterValue "/Landmass/Landscape/BlueprintBrushes/MPC/MPC_Erosion.MPC_Erosion" "SimulatedResolutionX" (Math|Conversions|ToFloat(Integer) _x_2) :!class /Script/Engine.KismetMaterialLibrary)
  (Rendering|Material|SetScalarParameterValue "/Landmass/Landscape/BlueprintBrushes/MPC/MPC_Erosion.MPC_Erosion" "SimulatedResolutionY" (Math|Conversions|ToFloat(Integer) _y_2) :!class /Script/Engine.KismetMaterialLibrary)
  (Rendering|Material|SetScalarParameterValue "/Landmass/Landscape/BlueprintBrushes/MPC/MPC_Erosion.MPC_Erosion" "ParticleSize" (.x (.scale (Variables|Debug|GetLandscapeTransform))) :!class /Script/Engine.KismetMaterialLibrary))

(fn ShowPreview
  (if (Variables|Debug|GetUse%20Landscape%20MID%20for%20Preview)
    (Development|ExecuteConsoleCommand "showflag.landscape 1")
    (Rendering|SetVisibility (Variables|Default|GetLandscape%20ISMC) false false)
    (bind _returnvalue (Landmass|GetLandscape))
    (Landscape|Runtime|Material|SetLandscapeMaterialTextureParameterValue _returnvalue "HeightRT" (Variables|Debug|RenderTargets|GetHeight%20RT%20A))
    (Landscape|Runtime|Material|SetLandscapeMaterialTextureParameterValue _returnvalue "NormalRT" (Variables|Debug|RenderTargets|GetNormal%20RT))
    (Landscape|Runtime|Material|SetLandscapeMaterialTextureParameterValue _returnvalue "SceneCaptureRT" (Variables|Debug|RenderTargets|GetSceneCapture%20RT))
    (Landscape|Runtime|Material|SetLandscapeMaterialScalarParameterValue _returnvalue "EnableLandmassPreview" 1.0)
    (Landscape|Runtime|Material|SetLandscapeMaterialScalarParameterValue _returnvalue "UseDerivedNormal" 0.0)
    (Landscape|Runtime|Material|SetLandscapeMaterialScalarParameterValue _returnvalue "RG8Height" 0.0)
    (else
      (Development|ExecuteConsoleCommand "showflag.landscape 0")
      (Rendering|SetVisibility (Variables|Default|GetLandscape%20ISMC) true false))))

(fn SimulateErosion
  (bind (_seconds _partialseconds) (Utilities|Time|GetAccurateRealTime))
  (Initialization|InitializeSimulation)
  (ErosionSettings|Reset)
  (Variables|Default|SetSimulating true)
  (Debug|SpawnInstanceLandscapeMeshes)
  (Initialization|InitializeErosionValues (Utilities|Array|Get(aref) (Variables|ErosionSettings|GetErosion%20Keyframes) (Variables|Default|GetIteration%20Key%20Index)))
  (bind _output_get (Variables|Default|SetStart%20Time (+ (Math|Conversions|ToFloat(Integer) _seconds) _partialseconds)))
  (Development|PrintString (Utilities|String|Append "Erosion Started at: " (Utilities|String|ToString(Float) _output_get) " seconds"))
  (Debug|ShowPreview))

(fn SingleIteration
  (CallFunction|ErosionParticles))

(fn SpawnInstanceLandscapeMeshes
  (bind (_x _y) (Math|Vector2D|BreakVector2D (Variables|Debug|GetWorldSize)))
  (bind (_origin _boxextent) (Math|BoxSphereBounds|BreakBoxSphereBounds (StaticMesh|GetBounds (Variables|/Script/Engine.StaticMeshComponent|GetStaticMesh (Variables|Default|GetLandscape%20ISMC)))))
  (bind (_x_2 _y_2) (Math|Vector2D|BreakVector2D (Variables|Debug|GetWorldSize)))
  (Utilities|FlowControl|Sequence
    (:then_0
      (Initialization|InitializeSimulation)
      (Components|InstancedStaticMesh|ClearInstances (Variables|Default|GetLandscape%20ISMC))
      (Development|PrintString "Erosion Brush Cleared Instances" false)
      (Utilities|FlowControl|Sequence
        (:then_0
          (bind (_current_point _total_points _0_1_phase _poly_center _x_row _y_row) (Utilities|2DGridExecutionMacro :Sector%20Size (/ (Math|Float|Max(Float) _x_2 _y_2) (Math|Float|Max(Float) (Math|Conversions|ToFloat(Integer) (Variables|Debug|GetSections%20X)) (Math|Conversions|ToFloat(Integer) (Variables|Debug|GetSections%20Y)))) :X (Variables|Debug|GetSections%20X) :Y (Variables|Debug|GetSections%20Y))
            (:Per%20Poly%20Execution
              (LocalVariables|SetCurrentMeshIndex _current_point)
              (Components|InstancedStaticMesh|AddInstance (Variables|Default|GetLandscape%20ISMC) (Math|Transform|MakeTransform (+ (* _poly_center "1.000000,1.000000, 0") (/ (+ (.location (Variables|Debug|GetLandscapeTransform)) (+ (.location (Variables|Debug|GetLandscapeTransform)) (Math|Conversions|ToVector(Vector2D) (Variables|Debug|GetWorldSize)))) 2.0)) "0, 0, 0" (Math|Vector|MakeVector (/ (/ _x (Math|Conversions|ToFloat(Integer) (Variables|Debug|GetSections%20X))) (* (.x _boxextent) 2.0)) (/ (/ _y (Math|Conversions|ToFloat(Integer) (Variables|Debug|GetSections%20Y))) (* (.y _boxextent) 2.0)) 0.001)) true)
              (Components|InstancedStaticMesh|SetCustomDataValue (Variables|Default|GetScatter%20Particles%20ISMC) (LocalVariables|GetCurrentMeshIndex) 0 (* (Math|Conversions|ToFloat(Integer) _x_row) (- (Variables|Debug|GetComponentRes) 1)) false)
              (Components|InstancedStaticMesh|SetCustomDataValue (Variables|Default|GetScatter%20Particles%20ISMC) (LocalVariables|GetCurrentMeshIndex) 1 (* (Math|Conversions|ToFloat(Integer) _y_row) (- (Variables|Debug|GetComponentRes) 1)) true))))))
    (:then_1
      (Initialization|InitializeErosionValues (Utilities|Array|Get(aref) (Variables|ErosionSettings|GetErosion%20Keyframes))))))

(fn SpawnInstanceParticleMeshes
  (params (ResolutionDivisor :type Integer))
  (bind (_x _y) (Math|Vector2D|BreakVector2D (Variables|Debug|GetWorldSize)))
  (bind (_origin _boxextent) (Math|BoxSphereBounds|BreakBoxSphereBounds (StaticMesh|GetBounds (Variables|/Script/Engine.StaticMeshComponent|GetStaticMesh (Variables|Default|GetScatter%20Particles%20ISMC)))))
  (bind (_x_2 _y_2) (Math|Vector2D|BreakVector2D (Variables|Debug|GetWorldSize)))
  (bind _scatter_particles_ismc (Variables|Default|GetScatter%20Particles%20ISMC))
  (Rendering|Material|SetScalarParameterValue "/Landmass/Landscape/BlueprintBrushes/MPC/MPC_Erosion.MPC_Erosion" "ResolutionDivisor" (Math|Conversions|ToFloat(Integer) ResolutionDivisor) :!class /Script/Engine.KismetMaterialLibrary)
  (Utilities|FlowControl|Sequence
    (:then_0
      (Components|InstancedStaticMesh|ClearInstances _scatter_particles_ismc)
      (Components|InstancedStaticMesh|SetNumCustomDataFloats _scatter_particles_ismc 2)
      (Development|PrintString "Cleared Instances")
      (Utilities|FlowControl|Sequence
        (:then_1
          (bind (_current_point _total_points _0_1_phase _poly_center _x_row _y_row) (Utilities|2DGridExecutionMacro :Sector%20Size (/ (Math|Float|Max(Float) _x_2 _y_2) (Math|Float|Max(Float) (Math|Conversions|ToFloat(Integer) (/ (Variables|Debug|GetSections%20X) (Variables|GetResolutionDivisor))) (Math|Conversions|ToFloat(Integer) (/ (Variables|Debug|GetSections%20Y) (Variables|GetResolutionDivisor))))) :X (/ (Variables|Debug|GetSections%20X) (Variables|GetResolutionDivisor)) :Y (/ (Variables|Debug|GetSections%20Y) (Variables|GetResolutionDivisor)))
            (:Per%20Poly%20Execution
              (LocalVariables|SetCurrentPoint _current_point)
              (Components|InstancedStaticMesh|AddInstance (Variables|Default|GetScatter%20Particles%20ISMC) (Math|Transform|MakeTransform (+ (* _poly_center "1.000000,1.000000, 0") (/ (+ (.location (Variables|Debug|GetLandscapeTransform)) (+ (.location (Variables|Debug|GetLandscapeTransform)) (Math|Conversions|ToVector(Vector2D) (Variables|Debug|GetWorldSize)))) 2.0)) "0, 0, 0" (Math|Vector|MakeVector (/ (/ _x (Math|Conversions|ToFloat(Integer) (/ (Variables|Debug|GetSections%20X) (Variables|GetResolutionDivisor)))) (* (.x _boxextent) 2.0)) (/ (/ _y (Math|Conversions|ToFloat(Integer) (/ (Variables|Debug|GetSections%20Y) (Variables|GetResolutionDivisor)))) (* (.y _boxextent) 2.0)) 1.0)) true)
              (Components|InstancedStaticMesh|SetCustomDataValue (Variables|Default|GetScatter%20Particles%20ISMC) (LocalVariables|GetCurrentPoint) 0 (* (Math|Conversions|ToFloat(Integer) _x_row) (- (Variables|Debug|GetComponentRes) 1)) false)
              (Components|InstancedStaticMesh|SetCustomDataValue (Variables|Default|GetScatter%20Particles%20ISMC) (LocalVariables|GetCurrentPoint) 1 (* (Math|Conversions|ToFloat(Integer) _y_row) (- (Variables|Debug|GetComponentRes) 1)) true))))))))

(fn ConstructionScript
  (bind _erosion_keyframes (Variables|ErosionSettings|GetErosion%20Keyframes))
  (Initialization|InitializeBrush)
  (Rendering|Material|SetScalarParameterValue "/Landmass/Landscape/BlueprintBrushes/MPC/MPC_Erosion.MPC_Erosion" "ShowWater" (Math|Conversions|ToFloat(Boolean) (Variables|Debug|GetShow%20Water)) :!class /Script/Engine.KismetMaterialLibrary)
  (Rendering|SetVisibility (Variables|Default|GetScatter%20Particles%20ISMC) (Variables|Debug|GetShow%20Water) false)
  (Variables|ErosionSettings|SetIterations%20Per%20Tick (Math|Integer|Min(Integer) (Variables|ErosionSettings|GetIterations%20Per%20Tick) 32))
  (for _index (range (+ (Utilities|Array|LastIndex _erosion_keyframes) 1))
    (Utilities|Struct|SetmembersinLandmassParticleErosionSettings (Utilities|Struct|BreakLandmassErosionKeyframe (Utilities|Array|Get(aref) _erosion_keyframes _index)) (Math|Integer|Clamp(Integer) (.ResolutionDivisor_30_9CEC792142F360E868D2129D125B4A4D (Utilities|Struct|BreakLandmassParticleErosionSettings (Utilities|Struct|BreakLandmassErosionKeyframe (Utilities|Array|Get(aref) _erosion_keyframes _index)))) 1 4))))
