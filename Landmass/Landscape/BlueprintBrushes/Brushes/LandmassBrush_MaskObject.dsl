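; Landmass landscape brush blueprint that acts purely as a shape mask for other uses such as erosion, rendering no layer of its own.
; EventSetShapeMask stores a texture in the edge falloff settings, which GetShapeMask reads back.
; ApplyMaskMIDParams pushes basic, edge falloff, noise and elevation mask settings to any supplied material instance.
; Construction and mask changes rebuild a BrushBounds preview material, flagged to draw to the entire landscape.
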(event EventBeginPlay)

(event EventTick
  (params (DeltaSeconds :type Float%20%28single-precision%29)))

(event EventCopyEdgeMaskParameters
  (params (MID :type Material%20Instance%20Dynamic%20Object%20Reference))
  (CallFunction|ApplyMaskMIDParams MID))

(event EventSetShapeMask
  (params (Texture :type Texture%202D%20Object%20Reference) (Points :type Array%20of%20Vectors) (NewExtents :type Vector%204%20Structure))
  (Utilities|Struct|SetmembersinLandmassEdgeFalloffSettings (Variables|MaskSettings|GetEdgeFalloff) Texture)
  (CallFunction|SetupPreviewMaterial))

(fn ApplyMaskMIDParams
  (params (MID :type Material%20Instance%20Dynamic%20Object%20Reference))
  (Utilities|SetBrushBasicMaterialParams MID)
  (Utilities|SetBrushEdgeFalloffMaterialParams MID (Variables|MaskSettings|GetEdgeFalloff))
  (Utilities|SetEdgeFalloffNoiseMaterialParams MID (Variables|MaskSettings|GetEdgeFalloffNoise))
  (Utilities|SetBrushElevationMaskParams MID (Variables|MaskSettings|GetElevationMasking)))

(fn GetShapeMask
  (return (Utilities|Struct|BreakLandmassEdgeFalloffSettings (Variables|MaskSettings|GetEdgeFalloff))))

(fn GetRenderDependencies
  (return 0))

(fn LandmassEditor
  (Default|LaunchLandmassEditor (Variables|Debug|GetBrushManager) self))

(fn RenderErosionLayer
  (params (In%20Parameters :type Landscape%20Brush%20Parameters%20Structure) (Extents :type Vector%204%20Structure) (BrushManager :type Landmass%20Brush%20Manager%20Object%20Reference))
  (return false))

(fn RenderLayer
  (params (InParameters :type Landscape%20Brush%20Parameters%20Structure%20%28by%20ref%29))
  ;[1 1 1] This Class in intended to act purely as a Shape Mask for other usage such as Erosion
  (return))

(fn SetupPreviewMaterial
  (Rendering|CreateAndAssignMID (Variables|Debug|GetExtentsPreviewMID) "/Landmass/Landscape/BlueprintBrushes/Materials/Internal/BrushBounds.BrushBounds" "HeightMID"
    (:Out
      (CallFunction|ApplyMaskMIDParams (Variables|Debug|GetExtentsPreviewMID))
      (Rendering|Material|SetScalarParameterValue (Variables|Debug|GetExtentsPreviewMID) "DrawToEntireLandscape" 1.0 :!class /Script/Engine.MaterialInstanceDynamic)
      (Default|SetMeshExentsMaterial (Variables|Debug|GetExtentsPreviewMID)))))

(fn ConstructionScript
  (CallFunction|SetupPreviewMaterial))
