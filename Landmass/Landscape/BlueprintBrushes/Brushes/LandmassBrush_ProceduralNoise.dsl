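; Landmass landscape brush that applies procedural noise to heightmaps and a named weightmap layer, shaped by edge falloff, elevation and slope masks.
; SetupMaterial, run on construction and shape-mask changes, creates a material instance for each target and pushes the settings into it.
; Weightmap noise can match the heightmap noise's tiling and seed.
; RenderLayer draws the instance matching the layer type, also feeding landscape height into weightmap passes.
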(event EventBeginPlay)

(event EventTick
  (params (DeltaSeconds :type Float%20%28single-precision%29)))

(event EventSetShapeMask
  (params (Texture :type Texture%202D%20Object%20Reference) (Points :type Array%20of%20Vectors) (NewExtents :type Vector%204%20Structure))
  (Utilities|Struct|SetmembersinLandmassEdgeFalloffSettings (Variables|HeightmapSettings|GetEdgeFalloff) Texture)
  (CallFunction|SetupMaterial))

(event EventCopyEdgeMaskParameters
  (params (MID :type Material%20Instance%20Dynamic%20Object%20Reference)))

(fn GetShapeMask
  (return (Utilities|Struct|BreakLandmassEdgeFalloffSettings (Variables|HeightmapSettings|GetEdgeFalloff))))

(fn GetRenderDependencies
  (return 0))

(fn LandmassEditor
  (Default|LaunchLandmassEditor (Variables|Debug|GetBrushManager) self))

(fn RenderErosionLayer
  (params (In%20Parameters :type Landscape%20Brush%20Parameters%20Structure) (Extents :type Vector%204%20Structure) (BrushManager :type Landmass%20Brush%20Manager%20Object%20Reference))
  (return false))

(fn RenderLayer
  (params (InParameters :type Landscape%20Brush%20Parameters%20Structure%20%28by%20ref%29))
  (bind (_renderareaworldtransform _renderareasize _combinedresult) (Utilities|Struct|BreakLandscapeBrushParameters (Variables|Debug|GetBrushRenderParameters)))
  (bind (_renderareaworldtransform_2 _renderareasize_2 _layertype) (Utilities|Struct|BreakLandscapeBrushParameters (Variables|Debug|GetBrushRenderParameters)))
  (bind (_renderareaworldtransform_3 _renderareasize_3 _combinedresult_2) (Utilities|Struct|BreakLandscapeBrushParameters (Variables|Debug|GetBrushRenderParameters)))
  (if (Utilities|Enum|Equal(Enum) _layertype "Heightmap")
    (Rendering|Material|SetTextureParameterValue (Variables|Debug|GetHeightmapRenderMID) "HeightRT" _combinedresult_2)
    (Default|DrawBrushMaterial (Variables|Debug|GetHeightmapRenderMID))
    (else
      (Rendering|Material|SetTextureParameterValue (Variables|Debug|GetWeightmapRenderMID) "WeightmapRT" _combinedresult)
      (bind _aslandmass_brush_manager (Utilities|Casting|CastToLandmassBrushManager (Variables|Debug|GetBrushManager)))
      (Rendering|Material|SetTextureParameterValue (Variables|Debug|GetWeightmapRenderMID) "HeightRT" (Variables|/Landmass/Landscape/BlueprintBrushes/LandmassBrushManager.LandmassBrushManager_C|GetInputLandscapeHeightRT _aslandmass_brush_manager))
      (Default|DrawBrushMaterial (Variables|Debug|GetWeightmapRenderMID))))
  (return))

(fn SetupMaterial
  (bind _noise_settings (Variables|HeightmapSettings|GetNoise%20Settings))
  (bind (_noisetiling_2_ef2d3d5242fb086d4d90c38c6b408450 _noiseseed_27_31da0b494e8b4b5bf9cbcf975b2b373c) (Utilities|Struct|BreakLandmassNoiseSettings _noise_settings))
  (bind _weightmap_noise_settings (Variables|WeightmapSettings|GetWeightmap%20Noise%20Settings))
  (bind (_noisemode_38_9ec6950a4ff4b2408377a98aecca9cf1 _noisetiling_2_ef2d3d5242fb086d4d90c38c6b408450_2 _noisecontrast_35_594d44164840da3129da148d47da1413 _noiseoctaves_6_94313b124adb34bc57e07589d88f4477 _noisepersistence_8_9d128c6549680567136bbbac7e1396bd _noiseseed_27_31da0b494e8b4b5bf9cbcf975b2b373c_2 _useturbulence_15_b728fba344672adcd12021b31426a480 _invert_16_dc2443514267e75d0a5e6b908033d0cb _noisefloor_47_3250694442b903dfcc8f80b36cc694e4 _noisemultiply_48_de34d15740c9580d77381b931e15aac9) (Utilities|Struct|BreakLandmassNoisePassWeightMapSettings _weightmap_noise_settings))
  (bind _weightmaprendermid (Variables|Debug|GetWeightmapRenderMID))
  (bind _weightmaprendermid_2 (Variables|Debug|GetWeightmapRenderMID))
  (bind _weightmaprendermid_3 (Variables|Debug|GetWeightmapRenderMID))
  (Utilities|FlowControl|Sequence
    (:then_0
      (Rendering|CreateAndAssignMID (Variables|Debug|GetHeightmapRenderMID) (Variables|HeightmapSettings|GetHeightMaterial) "HeightMID"
        (:Out
          (Utilities|SetBrushBasicMaterialParams (Variables|Debug|GetHeightmapRenderMID))
          (Utilities|SetBrushEdgeFalloffMaterialParams (Variables|Debug|GetHeightmapRenderMID) (Variables|HeightmapSettings|GetEdgeFalloff))
          (Utilities|SetEdgeFalloffNoiseMaterialParams (Variables|Debug|GetHeightmapRenderMID) (Variables|HeightmapSettings|GetEdgeFalloffNoise))
          (Utilities|SetBrushNoiseMaterialParams (Variables|Debug|GetHeightmapRenderMID) (Variables|HeightmapSettings|GetNoise%20Settings))
          (Utilities|SetElevationMaskParams (Variables|Debug|GetHeightmapRenderMID) (Variables|HeightmapSettings|GetHeightmap%20Elevation%20Mask))
          (Utilities|SetSlopeMaskParams (Variables|Debug|GetHeightmapRenderMID) (Variables|HeightmapSettings|GetHeightmap%20Slope%20Mask)))))
    (:then_1
      (Rendering|CreateAndAssignMID (Variables|Debug|GetWeightmapRenderMID) (Variables|WeightmapSettings|GetWeightmapMaterial)
        (:Out
          (Utilities|SetBrushBasicMaterialParams (Variables|Debug|GetWeightmapRenderMID) false)
          (Utilities|SetBrushEdgeFalloffMaterialParams (Variables|Debug|GetWeightmapRenderMID) (Variables|HeightmapSettings|GetEdgeFalloff))
          (Utilities|SetEdgeFalloffNoiseMaterialParams (Variables|Debug|GetWeightmapRenderMID) (Variables|HeightmapSettings|GetEdgeFalloffNoise))
          (Utilities|SetElevationMaskParams (Variables|Debug|GetWeightmapRenderMID) (Variables|WeightmapSettings|GetWeightmap%20Elevation%20Mask))
          (Utilities|SetSlopeMaskParams (Variables|Debug|GetWeightmapRenderMID) (Variables|WeightmapSettings|GetWeightmap%20Slope%20Mask)))))
    (:then_2
      (Utilities|FlowControl|Sequence
        (:then_0
          (Rendering|Material|SetScalarParameterValue _weightmaprendermid "WeightmapOpacity" (Variables|WeightmapSettings|GetOpacity) :!class /Script/Engine.MaterialInstanceDynamic)
          (Rendering|Material|SetScalarParameterValue _weightmaprendermid "FalloffWidth" (Math|Float|Max(Float) (Variables|WeightmapSettings|GetFalloff%20Width) 0.0001) :!class /Script/Engine.MaterialInstanceDynamic)
          (Rendering|Material|SetScalarParameterValue _weightmaprendermid "FalloffOffset" (Variables|WeightmapSettings|GetEdge%20Offset) :!class /Script/Engine.MaterialInstanceDynamic))
        (:then_1
          (Utilities|FlowControl|Sequence
            (:then_0
              (Rendering|Material|SetScalarParameterValue _weightmaprendermid_2 "MultiplyNoise" (Math|Conversions|ToFloat(Boolean) (Utilities|Enum|Equal(Enum) _noisemode_38_9ec6950a4ff4b2408377a98aecca9cf1 "NewEnumerator1")) :!class /Script/Engine.MaterialInstanceDynamic)
              (Rendering|Material|SetScalarParameterValue _weightmaprendermid_2 "WeightNoiseTiling" _noisetiling_2_ef2d3d5242fb086d4d90c38c6b408450_2 :!class /Script/Engine.MaterialInstanceDynamic)
              (Rendering|Material|SetScalarParameterValue _weightmaprendermid_2 "WeightNoiseContrast" _noisecontrast_35_594d44164840da3129da148d47da1413 :!class /Script/Engine.MaterialInstanceDynamic)
              (Rendering|Material|SetScalarParameterValue _weightmaprendermid_2 "WeightNoiseOctaves" _noiseoctaves_6_94313b124adb34bc57e07589d88f4477 :!class /Script/Engine.MaterialInstanceDynamic)
              (Rendering|Material|SetScalarParameterValue _weightmaprendermid_2 "WeightNoisePersistence" _noisepersistence_8_9d128c6549680567136bbbac7e1396bd :!class /Script/Engine.MaterialInstanceDynamic)
              (Rendering|Material|SetScalarParameterValue _weightmaprendermid_2 "WeightNoiseSeed" _noiseseed_27_31da0b494e8b4b5bf9cbcf975b2b373c_2 :!class /Script/Engine.MaterialInstanceDynamic)
              (Rendering|Material|SetScalarParameterValue _weightmaprendermid_2 "WeightNoiseTurbulence" (Math|Conversions|ToFloat(Boolean) _useturbulence_15_b728fba344672adcd12021b31426a480) :!class /Script/Engine.MaterialInstanceDynamic)
              (Rendering|Material|SetScalarParameterValue _weightmaprendermid_2 "WeightInvertNoise" (Math|Conversions|ToFloat(Boolean) _invert_16_dc2443514267e75d0a5e6b908033d0cb) :!class /Script/Engine.MaterialInstanceDynamic)
              (Rendering|Material|SetScalarParameterValue _weightmaprendermid_2 "WeightNoiseFloor" _noisefloor_47_3250694442b903dfcc8f80b36cc694e4 :!class /Script/Engine.MaterialInstanceDynamic)
              (Rendering|Material|SetScalarParameterValue _weightmaprendermid_2 "WeightNoiseMultiply" _noisemultiply_48_de34d15740c9580d77381b931e15aac9 :!class /Script/Engine.MaterialInstanceDynamic))
            (:then_1
              (if (Variables|WeightmapSettings|GetMatch%20Heightmap%20Noise)
                (Rendering|Material|SetScalarParameterValue _weightmaprendermid_3 "WeightNoiseTiling" _noisetiling_2_ef2d3d5242fb086d4d90c38c6b408450 :!class /Script/Engine.MaterialInstanceDynamic)
                (Rendering|Material|SetScalarParameterValue _weightmaprendermid_3 "WeightNoiseSeed" _noiseseed_27_31da0b494e8b4b5bf9cbcf975b2b373c :!class /Script/Engine.MaterialInstanceDynamic)))))))))

(fn ConstructionScript
  (CallFunction|SetupMaterial)
  (Variables|Debug|SetWeightmapLayers (Utilities|Array|MakeArray (Variables|WeightmapSettings|GetWeightmap%20Name)))
  (if (or (not (Variables|BrushSettings|GetAffectsHeightmap)) (not (Variables|BrushSettings|GetAffectsWeightmaps)))
    (Default|RequestUpdatefromBrush (Variables|Debug|GetBrushManager) self)))
