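; Landmass landscape brush that, despite its Terraces name, acts purely as a shape mask for other uses such as erosion.
; Its layer and erosion render functions draw nothing, and an author comment confirms the mask-only intent.
; Construction builds a bounds-preview material from edge falloff, noise and elevation masking settings, then updates the mask.
; SetShapeMask stores a texture in the falloff settings and rebuilds that preview.
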(event EventBeginPlay)

(event EventTick
  (params (DeltaSeconds :type Float%20%28single-precision%29)))

(event EventSetShapeMask
  (params (Texture :type Texture%202D%20Object%20Reference) (Points :type Array%20of%20Vectors) (NewExtents :type Vector%204%20Structure))
  (Utilities|Struct|SetmembersinLandmassEdgeFalloffSettings (Variables|MaskSettings|GetEdgeFalloff) Texture)
  (CallFunction|SetupPreviewMaterial))

(event Custom|Update%20Mask
  (Default|CallMaskPropertiesUpdate))

(fn ApplyMaskMIDParams
  (params (MID :type Material%20Instance%20Dynamic%20Object%20Reference))
  (Utilities|SetBrushBasicMaterialParams MID)
  (Utilities|SetBrushEdgeFalloffMaterialParams MID (Variables|MaskSettings|GetEdgeFalloff))
  (Utilities|SetEdgeFalloffNoiseMaterialParams MID (Variables|MaskSettings|GetEdgeFalloffNoise))
  (Utilities|SetBrushElevationMaskParams MID (Variables|MaskSettings|GetElevationMasking)))

(fn GetShapeMask
  (return (Utilities|Struct|BreakLandmassEdgeFalloffSettings (Variables|MaskSettings|GetEdgeFalloff))))

(fn GetRenderDependencies
  (return 0))

(fn LandmassEditor
  (Default|LaunchLandmassEditor (Variables|Debug|GetBrushManager) self))

(fn RenderErosionLayer
  (params (In%20Parameters :type Landscape%20Brush%20Parameters%20Structure) (Extents :type Vector%204%20Structure) (BrushManager :type Landmass%20Brush%20Manager%20Object%20Reference))
  (return false))

(fn RenderLayer
  (params (InParameters :type Landscape%20Brush%20Parameters%20Structure%20%28by%20ref%29))
  ;[1 1 1] This Class in intended to act purely as a Shape Mask for other usage such as Erosion
  (return))

(fn SetupPreviewMaterial
  (bind _planemid (Variables|Default|GetPlaneMID))
  (Rendering|CreateAndAssignMID (Variables|Default|GetPlaneMID) "/Landmass/Landscape/BlueprintBrushes/Materials/Internal/BrushBounds_ShapeMaskPreview.BrushBounds_ShapeMaskPreview" "HeightMID"
    (:Out
      (CallFunction|ApplyMaskMIDParams (Variables|Default|GetPlaneMID))
      (Rendering|Material|SetTextureParameterValue _planemid "ShapeRT" (Utilities|Struct|BreakLandmassEdgeFalloffSettings (Variables|MaskSettings|GetEdgeFalloff)))
      (Rendering|Material|SetScalarParameterValue _planemid "DrawToEntireLandscape" 1.0 :!class /Script/Engine.MaterialInstanceDynamic)
      (Default|SetMeshExentsMaterial (Variables|Default|GetPlaneMID)))))

(fn ConstructionScript
  (CallFunction|SetupPreviewMaterial)
  (CallFunction|UpdateMask))
