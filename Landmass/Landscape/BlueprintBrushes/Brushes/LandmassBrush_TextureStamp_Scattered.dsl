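; Landmass landscape brush that stamps a displacement texture into heightmaps and weightmaps, with repeat, random offset and seed parameters for scattering.
; Construction builds a heightmap MID and one MID per weightmap layer, pruning stale ones and pushing falloff, displacement and custom parameters.
; RenderLayer binds the landscape render target, draws the MID for that layer type, then draws each configured noise pass on heightmaps only.
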(event EventBeginPlay)

(event Collision|EventActorBeginOverlap
  (params (OtherActor :type Actor%20Object%20Reference)))

(event EventSetShapeMask
  (params (Texture :type Texture%202D%20Object%20Reference) (Points :type Array%20of%20Vectors) (NewExtents :type Vector%204%20Structure))
  (Utilities|Struct|SetmembersinLandmassEdgeFalloffSettings (Variables|HeightmapSettings|GetEdgeFalloff) Texture)
  (Default|SetupBrushMIDs))

(event EventCopyEdgeMaskParameters
  (params (MID :type Material%20Instance%20Dynamic%20Object%20Reference))
  (Utilities|SetBrushBasicMaterialParams MID)
  (Utilities|SetBrushEdgeFalloffMaterialParams MID (Variables|HeightmapSettings|GetEdgeFalloff))
  (Utilities|SetEdgeFalloffNoiseMaterialParams MID (Variables|HeightmapSettings|GetEdgeFalloffNoise)))

(fn AssignRenderTargets
  (bind (_renderareaworldtransform _renderareasize _combinedresult) (Utilities|Struct|BreakLandscapeBrushParameters (Variables|Debug|GetBrushRenderParameters)))
  (bind (_renderareaworldtransform_2 _renderareasize_2 _layertype) (Utilities|Struct|BreakLandscapeBrushParameters (Variables|Debug|GetBrushRenderParameters)))
  (bind (_renderareaworldtransform_3 _renderareasize_3 _combinedresult_2 _layertype_2 _weightmaplayername) (Utilities|Struct|BreakLandscapeBrushParameters (Variables|Debug|GetBrushRenderParameters)))
  (bind (_renderareaworldtransform_4 _renderareasize_4 _combinedresult_3) (Utilities|Struct|BreakLandscapeBrushParameters (Variables|Debug|GetBrushRenderParameters)))
  (if (Utilities|Enum|Equal(Enum) _layertype "Heightmap")
    (Rendering|Material|SetTextureParameterValue (Variables|Debug|GetHeightmapRenderMID) "HeightRT" _combinedresult)
    (else
      (Rendering|Material|SetTextureParameterValue (Utilities|Map|Find (Variables|Debug|GetDraw_Weightmap_MIDMap) _weightmaplayername) "WeightmapRT" _combinedresult_3))))

(fn DrawBrush
  (bind (_renderareaworldtransform _renderareasize _layertype) (Utilities|Struct|BreakLandscapeBrushParameters (Variables|Debug|GetBrushRenderParameters)))
  (bind (_renderareaworldtransform_2 _renderareasize_2 _combinedresult _layertype_2 _weightmaplayername) (Utilities|Struct|BreakLandscapeBrushParameters (Variables|Debug|GetBrushRenderParameters)))
  (if (Utilities|Enum|Equal(Enum) _layertype "Heightmap")
    (Default|DrawBrushMaterial (Variables|Debug|GetHeightmapRenderMID))
    (else
      (Default|DrawBrushMaterial (Utilities|Map|Find (Variables|Debug|GetDraw_Weightmap_MIDMap) _weightmaplayername)))))

(fn DrawNoiseMaterials
  (bind (_renderareaworldtransform _renderareasize _combinedresult) (Utilities|Struct|BreakLandscapeBrushParameters (Variables|Debug|GetBrushRenderParameters)))
  (bind (_renderareaworldtransform_2 _renderareasize_2 _combinedresult_2 _layertype) (Utilities|Struct|BreakLandscapeBrushParameters (Variables|Debug|GetBrushRenderParameters)))
  (bind _addnoisemid (Variables|Debug|GetAddNoiseMID))
  (bind _addnoisemid_2 (Variables|Debug|GetAddNoiseMID))
  (if (Utilities|Enum|Equal(Enum) _layertype "Heightmap")
    (if (> (Utilities|Array|Length (Variables|NoisePasses|GetNoise%20Passes)) 0)
      (Rendering|CreateAndAssignMID (Variables|Debug|GetAddNoiseMID) "/Landmass/Landscape/BlueprintBrushes/Materials/BrushMaterials/M_Brush_NoisePass.M_Brush_NoisePass"
        (:Out
          (Default|SetMaterialParams (Variables|Debug|GetAddNoiseMID))
          (Rendering|Material|SetTextureParameterValue (Variables|Debug|GetAddNoiseMID) "HeightRT" _combinedresult)
          (for _array_element (Variables|NoisePasses|GetNoise%20Passes)
            (Utilities|FlowControl|Sequence
              (:then_0
                (Rendering|Material|SetScalarParameterValue _addnoisemid_2 "NoiseTiling" (Utilities|Struct|BreakLandmassNoisePassHeightSettings _array_element) :!class /Script/Engine.MaterialInstanceDynamic)
                (Rendering|Material|SetScalarParameterValue _addnoisemid_2 "NoiseHeight" (.NoiseHeight_4_594D44164840DA3129DA148D47DA1413 (Utilities|Struct|BreakLandmassNoisePassHeightSettings _array_element)) :!class /Script/Engine.MaterialInstanceDynamic)
                (Rendering|Material|SetScalarParameterValue _addnoisemid_2 "NoiseOctaves" (.NoiseOctaves_6_94313B124ADB34BC57E07589D88F4477 (Utilities|Struct|BreakLandmassNoisePassHeightSettings _array_element)) :!class /Script/Engine.MaterialInstanceDynamic)
                (Rendering|Material|SetScalarParameterValue _addnoisemid_2 "NoisePersistence" (.NoisePersistence_8_9D128C6549680567136BBBAC7E1396BD (Utilities|Struct|BreakLandmassNoisePassHeightSettings _array_element)) :!class /Script/Engine.MaterialInstanceDynamic)
                (Rendering|Material|SetScalarParameterValue _addnoisemid_2 "NoiseFloor" (.NoiseFloor_30_F1F44BD74164C4B6ECE8A7AF966A44F9 (Utilities|Struct|BreakLandmassNoisePassHeightSettings _array_element)) :!class /Script/Engine.MaterialInstanceDynamic)
                (Rendering|Material|SetScalarParameterValue _addnoisemid_2 "NoiseSeed" (.NoiseSeed_27_31DA0B494E8B4B5BF9CBCF975B2B373C (Utilities|Struct|BreakLandmassNoisePassHeightSettings _array_element)) :!class /Script/Engine.MaterialInstanceDynamic)
                (Rendering|Material|SetScalarParameterValue _addnoisemid "UseNoiseTurbulence" (Math|Conversions|ToFloat(Boolean) (.UseTurbulence_15_B728FBA344672ADCD12021B31426A480 (Utilities|Struct|BreakLandmassNoisePassHeightSettings _array_element))) :!class /Script/Engine.MaterialInstanceDynamic)
                (Rendering|Material|SetScalarParameterValue _addnoisemid "InvertNoise" (Math|Conversions|ToFloat(Boolean) (.Invert_16_DC2443514267E75D0A5E6B908033D0CB (Utilities|Struct|BreakLandmassNoisePassHeightSettings _array_element))) :!class /Script/Engine.MaterialInstanceDynamic)
                (Rendering|Material|SetScalarParameterValue _addnoisemid "FalloffOffset" (.NoiseEdgeOffset_23_50E546C145067F2DC0F3FD8D1605471F (Utilities|Struct|BreakLandmassNoisePassHeightSettings _array_element)) :!class /Script/Engine.MaterialInstanceDynamic)
                (Rendering|Material|SetScalarParameterValue _addnoisemid "FalloffWidth" (.NoiseFalloffWidth_24_8D344DAF4B7F6800D5F05F8491EEACA9 (Utilities|Struct|BreakLandmassNoisePassHeightSettings _array_element)) :!class /Script/Engine.MaterialInstanceDynamic)
                (Rendering|Material|SetScalarParameterValue _addnoisemid "IntersectionSmoothness" (.IntersectionSmoothness_33_4E4F477745B7C540CD7340AE1F9A839F (Utilities|Struct|BreakLandmassNoisePassHeightSettings _array_element)) :!class /Script/Engine.MaterialInstanceDynamic))
              (:then_1
                (Default|DrawBrushMaterial (Variables|Debug|GetAddNoiseMID))))))))))

(fn GetShapeMask
  (return (Utilities|Struct|BreakLandmassEdgeFalloffSettings (Variables|HeightmapSettings|GetEdgeFalloff))))

(fn GetRenderDependencies
  (return 0))

(fn LandmassEditor
  (Default|LaunchLandmassEditor (Variables|Debug|GetBrushManager) self))

(fn RenderErosionLayer
  (params (In%20Parameters :type Landscape%20Brush%20Parameters%20Structure) (Extents :type Vector%204%20Structure) (BrushManager :type Landmass%20Brush%20Manager%20Object%20Reference))
  (return false))

(fn RenderLayer
  (params (InParameters :type Landscape%20Brush%20Parameters%20Structure%20%28by%20ref%29))
  ;[1 0.9016 0.1562] Assign RT to MID
  (CallFunction|AssignRenderTargets)
  ;[0.0349 0.4746 1] Draw the Material
  (CallFunction|DrawBrush)
  ;[0.08092 0.8177 0.4013] Render Noise Passes
  (CallFunction|DrawNoiseMaterials)
  ;[0.1699 0.01189 0.25] Complete
  (return))

(fn SetMaterialParams
  (params (In%20MID :type Material%20Instance%20Dynamic%20Object%20Reference) (IsHeightmap :type Boolean))
  (bind _in_mid (Variables|GetIn%20MID))
  (bind _in_mid_2 (Variables|GetIn%20MID))
  (Utilities|IsValid (Variables|Debug|GetBrushManager)
    (:Is%20Valid
      (Utilities|FlowControl|Sequence
        (:then_0
          ;[1 1 1] Basic Brush Parameters
          (Utilities|SetBrushBasicMaterialParams (Variables|GetIn%20MID) (Variables|GetIsHeightmap)))
        (:then_1
          ;[1 1 1] Displacement and Falloff
          (Utilities|SetBrushEdgeFalloffMaterialParams (Variables|GetIn%20MID) (Variables|HeightmapSettings|GetEdgeFalloff))
          (Utilities|SetEdgeFalloffNoiseMaterialParams (Variables|GetIn%20MID) (Variables|HeightmapSettings|GetEdgeFalloffNoise))
          (Rendering|Material|SetScalarParameterValue _in_mid "RelativeDepth" (Variables|HeightmapSettings|GetRelative%20Depth) :!class /Script/Engine.MaterialInstanceDynamic)
          (Rendering|Material|SetTextureParameterValue _in_mid "Displacement" (select (Utilities|IsValid (Variables|HeightmapSettings|GetDisplacement) :!class /Script/Engine.KismetSystemLibrary) (Variables|HeightmapSettings|GetDisplacement) "/Engine/ArtTools/RenderToTexture/Textures/T_EV_BlankWhite_01.T_EV_BlankWhite_01"))
          (Rendering|Material|SetScalarParameterValue _in_mid "DisplacementAsAlpha" (Utilities|Select :Index (Variables|HeightmapSettings|DisplacementSettings|GetAlphablend%20Behavior) :NewEnumerator0 1.0 :NewEnumerator1 0.0) :!class /Script/Engine.MaterialInstanceDynamic)
          (Rendering|Material|SetVectorParameterValue _in_mid "DisplacementChannel" (Utilities|Select :Index (Variables|HeightmapSettings|DisplacementSettings|GetTexture%20Channel) :NewEnumerator0 "(R=1.000000,G=0.000000,B=0.000000,A=0.000000)" :NewEnumerator1 "(R=0.000000,G=1.000000,B=0.000000,A=0.000000)" :NewEnumerator2 "(R=0.000000,G=0.000000,B=1.000000,A=0.000000)" :NewEnumerator3 "(R=0.000000,G=0.000000,B=0.000000,A=1.000000)") :!class /Script/Engine.MaterialInstanceDynamic)
          (Rendering|Material|SetScalarParameterValue _in_mid "DisplacementReferencePlane" (Variables|HeightmapSettings|DisplacementSettings|GetHeight%20Reference%20Plane) :!class /Script/Engine.MaterialInstanceDynamic)
          (Rendering|Material|SetScalarParameterValue _in_mid_2 "CellOffset" (Variables|HeightmapSettings|GetRandom%20Offset) :!class /Script/Engine.MaterialInstanceDynamic)
          (Rendering|Material|SetScalarParameterValue _in_mid_2 "CellSeed" (Variables|HeightmapSettings|GetRandom%20Seed) :!class /Script/Engine.MaterialInstanceDynamic)
          (Rendering|Material|SetScalarParameterValue _in_mid_2 "Opacity" (Variables|HeightmapSettings|GetOpacity) :!class /Script/Engine.MaterialInstanceDynamic)
          (Rendering|Material|SetScalarParameterValue _in_mid_2 "Repeats" (Variables|HeightmapSettings|GetStamp%20Repeats) :!class /Script/Engine.MaterialInstanceDynamic))
        (:then_2
          ;[1 1 1] Apply Custom Parameters: Allows customizing without making a new BP Class
          (Class|LandmassFunctionLibrary|ApplyCustomMaterialParams (Variables|GetIn%20MID) (Variables|HeightmapSettings|CustomParameters|GetCustom%20Scalar%20Params) (Variables|HeightmapSettings|CustomParameters|GetCustom%20Vector%20Params) (Variables|HeightmapSettings|CustomParameters|GetCustom%20Texture%20Params)))))))

(fn SetupBrushMIDs
  (bind (_opacity_20_d0865a8c44e92ecd33a766b635b90848 _falloffwidth_5_ef2fba6242362acba8ad7a902f5c8615 _edgeoffset_6_89aa741b44dfdb769c3c158824c6462b _materialoverride_8_729913b44c7d08cc7f683593d76b0be3) (Utilities|Struct|BreakLandmassWeightmapSettings (Utilities|Map|Find (Variables|WeightmapSettings|GetWeightmap%20Layers) (Variables|GetCurrentWeightmap))))
  (bind _currentweightmap (Variables|GetCurrentWeightmap))
  (bind _draw_weightmap_midmap (Variables|Debug|GetDraw_Weightmap_MIDMap))
  (bind (_value _returnvalue) (Utilities|Map|Find _draw_weightmap_midmap _currentweightmap))
  (bind _curindex (Variables|GetCurIndex))
  (bind _draw_weightmap_midmap_2 (Variables|Debug|GetDraw_Weightmap_MIDMap))
  (Utilities|FlowControl|Sequence
    (:then_0
      (Rendering|CreateAndAssignMID (Variables|Debug|GetHeightmapRenderMID) (Variables|HeightmapSettings|GetHeightMaterial)
        (:Out
          (Default|SetMaterialParams (Variables|Debug|GetHeightmapRenderMID) true)
          (Utilities|SetupBrushPreviewShape (Variables|HeightmapSettings|GetEdgeFalloff) (Variables|HeightmapSettings|GetEdgeFalloffNoise)))))
    (:then_1
      ; Allocate Weightmap MIDs
      (Utilities|FlowControl|Sequence
        (:then_0
          (bind _keys (Utilities|Map|Keys (Variables|WeightmapSettings|GetWeightmap%20Layers)))
          (for _array_element _keys
            (Variables|SetCurrentWeightmap _array_element)
            (if _returnvalue
              (Variables|SetTempMID _value)
              (else
                (Variables|SetTempMID 0)))
            (Rendering|CreateAndAssignMID (Variables|GetTempMID) (select (Utilities|IsValid _materialoverride_8_729913b44c7d08cc7f683593d76b0be3 :!class /Script/Engine.KismetSystemLibrary) _materialoverride_8_729913b44c7d08cc7f683593d76b0be3 (Variables|WeightmapSettings|GetWeightmapMaterial))
              (:Out
                (Utilities|Map|Add (Variables|Debug|GetDraw_Weightmap_MIDMap) (Variables|GetCurrentWeightmap) (Variables|GetTempMID)))))
          (for _index (range (Utilities|Map|Length (Variables|Debug|GetDraw_Weightmap_MIDMap)))
            (Variables|SetCurIndex _index)
            (bind _keys_2 (Utilities|Map|Keys (Variables|Debug|GetDraw_Weightmap_MIDMap)))
            (if (not (.ReturnValue (Utilities|Map|Find (Variables|WeightmapSettings|GetWeightmap%20Layers) (Utilities|Array|Get(aref) _keys_2 (- (Utilities|Array|LastIndex _keys_2) _curindex)))))
              (Utilities|Map|Remove (Variables|Debug|GetDraw_Weightmap_MIDMap) (Utilities|Array|Get(aref) _keys_2 (- (Utilities|Array|LastIndex _keys_2) _curindex))))))
        (:then_1
          _))
      (bind _keys_3 (Utilities|Map|Keys (Variables|Debug|GetDraw_Weightmap_MIDMap)))
      (for _array_element_2 _keys_3
        (Default|SetMaterialParams (Utilities|Map|Find _draw_weightmap_midmap_2 _array_element_2))
        (Utilities|SetBrushWeightmapParams (Utilities|Map|Find _draw_weightmap_midmap_2 _array_element_2) (Utilities|Map|Find (Variables|WeightmapSettings|GetWeightmap%20Layers) _array_element_2))))))

(fn ConstructionScript
  (Utilities|SkipifTransientObject self
    (:Exectute
      (bind _keys (Utilities|Map|Keys (Variables|WeightmapSettings|GetWeightmap%20Layers)))
      (Variables|Debug|SetWeightmapLayers _keys)
      (Default|SetupBrushMIDs))))
